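# Niagara system for the PistolCanary tentacle energy ball's charge beam, stretched from the system origin to User.Stop Location.
# Ribbon spawn groups follow a Hermite spline, with per-group width masks layering core, glow and detail under pulsing waveforms.
# Muzzle flares, shell sprites, rotating energy meshes, an impact flare and GPU center flares, one occlusion-tested, add glow.
# User parameters drive scale and duration; a disappear timer fades layers, and curl-noise end particles scatter along the beam.
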NiagaraSystem /PistolCanary/Assets/FX/TentacleEnergyBall/NS_PC_EnergyBall_Charge_Beam
dependencies {
 usedDataInterfaces [
  &NiagaraDataInterfaceCurve
  &NiagaraDataInterfaceVector2DCurve
  &NiagaraDataInterfaceColorCurve
  &NiagaraDataInterfaceCamera
  /Script/Niagara.NiagaraDataInterfaceOcclusion
 ]
 usedDynamicInputs [
  /Niagara/DynamicInputs/Helpers/SimulationPosition
  &TransformPosition
  &VectorLength
  &RemapRange
  &ReturnCameraPosition
  &SetBoolByFloatComparison
  &Add_Float
 ]
 usedModules [
  &SystemState
  &Timeline
  &InterpolateOverTime
  &self:SystemSpawnScript.SystemScriptSource.SystemScriptGraph.NiagaraNodeAssignment_0.SetVariables_#1
  &self:SystemSpawnScript.SystemScriptSource.SystemScriptGraph.NiagaraNodeAssignment_10.SetVariables_#2
  &self:SystemSpawnScript.SystemScriptSource.SystemScriptGraph.NiagaraNodeAssignment_2.SetVariables_#3
  &self:SystemSpawnScript.SystemScriptSource.SystemScriptGraph.NiagaraNodeAssignment_3.SetVariables_#4
  &self:SystemSpawnScript.SystemScriptSource.SystemScriptGraph.NiagaraNodeAssignment_4.SetVariables_#5
  &self:SystemSpawnScript.SystemScriptSource.SystemScriptGraph.NiagaraNodeAssignment_5.SetVariables_#6
  &self:SystemSpawnScript.SystemScriptSource.SystemScriptGraph.NiagaraNodeAssignment_6.SetVariables_#7
  &self:SystemSpawnScript.SystemScriptSource.SystemScriptGraph.NiagaraNodeAssignment_1.SetVariables_#8
  &self:SystemSpawnScript.SystemScriptSource.SystemScriptGraph.NiagaraNodeAssignment_7.SetVariables_#9
  &self:SystemSpawnScript.SystemScriptSource.SystemScriptGraph.NiagaraNodeAssignment_8.SetVariables_#10
  &self:SystemSpawnScript.SystemScriptSource.SystemScriptGraph.NiagaraNodeAssignment_9.SetVariables_#11
  &DoOnce
  &self:SystemSpawnScript.SystemScriptSource.SystemScriptGraph.NiagaraNodeAssignment_12.SetVariables_#12
 ]
}
emitters {
 Beams {
  emitterUpdateScript [
   &EmitterState()
   &SpawnBurst_Instantaneous(Spawn Count=20, Use Spawn Probability=false, UseLoopCountLimit=false)
   &SpawnBurst_Instantaneous(Spawn Count=20, Spawn Group=1, Use Spawn Probability=false, UseLoopCountLimit=false) {module=SpawnBurst_Instantaneous001}
   &SpawnBurst_Instantaneous(Spawn Count=20, Spawn Group=2, Use Spawn Probability=false, UseLoopCountLimit=false) {module=SpawnBurst_Instantaneous002}
   &SpawnBurst_Instantaneous(Spawn Count=20, Spawn Group=3, Use Spawn Probability=false, UseLoopCountLimit=false) {module=SpawnBurst_Instantaneous003}
   &SpawnBurst_Instantaneous(Spawn Count=20, Spawn Group=4, Use Spawn Probability=false, UseLoopCountLimit=false) {module=SpawnBurst_Instantaneous004}
   &SpawnBurst_Instantaneous(Spawn Count=20, Spawn Group=5, Use Spawn Probability=false, UseLoopCountLimit=false) {module=SpawnBurst_Instantaneous005}
   /SetVariables(Emitter.ScaleW:NiagaraFloat=&Waveform("Waveform [1]"=::NewEnumerator12, "[1] Frequency"=20, "[1] Amplitude Min/Max"=<0 1>)) {isSetParameters=true}
   /SetVariables(Emitter.BeamWidthScale:NiagaraFloat=&FloatFromCurve(
     FloatCurve=DataInterface() {propertyValues {Curve {
        defaultValue=1
        keys [
         {tangentMode=RCTM_User, time=0, value=0}
         {arriveTangent=1.25, interpMode=RCIM_Cubic, leaveTangent=1.25, time=0.05, value=1}
         {arriveTangent=-0.6517, interpMode=RCIM_Cubic, leaveTangent=-0.6517, tangentMode=RCTM_User, time=0.8, value=1}
         {arriveTangent=-3.4192, interpMode=RCIM_Cubic, leaveTangent=-3.4192, tangentMode=RCTM_User, time=1, value=0}
        ]
       }}}
     CurveIndex=$Emitter.NormalizedLoopAge
     Scale Curve=$User.BeamWidth
    )) {isSetParameters=true}
   &SpawnBurst_Instantaneous(Spawn Count=20, Spawn Time=1.4833, Use Spawn Probability=false, UseLoopCountLimit=false) {module=SpawnBurst_Instantaneous006}
  ]
  particleSpawnScript [
   &PartitionParticles(
    Number Of Partitions=1
    Particles Per Partition=20
    Partition Mode:ENiagara_PartitionMode_v2=::NewEnumerator1
    Write Partition Index to Ribbon Link Order:ENiagara_AttributeSamplingApplyOutput=::NewEnumerator0
   )
   /SetVariables(Particles.RibbonID=/Niagara/DynamicInputs/Meshes/MakeRibbonIDFromSpawnGroup()) {isSetParameters=true}
   &InitializeParticle(
    Color=&MergeHelper_ColorSwitch(
     0=<rgba 0.576 0 1 1>
     1=<rgba 0.6 0.6 0.6 1.5>
     2=<rgba 0 0 0 0>
     3=<rgba 0.05 0.0005 0.0046 0.1>
     4=<rgba 0 0 0 0>
     5=<rgba 0 0 0 0>
     Index=&SpawnGroup()
     Num=::NewEnumerator4
    )
    Lifetime=$System.CurrentLoopDuration
    Ribbon Width=&MergeHelper_FloatSwitch(0=200, 1=90, 2=140, 3=700.3717, 4=170, 5=100, Index=&SpawnGroup(), Num=::NewEnumerator4)
    Ribbon Width Mode=::NewEnumerator1
    UsePositionOffset=false
    Write Material Random=true
   )
  ]
  particleUpdateScript [
   &ParticleState()
   /Game/Effects/Niagara/Modules/Curves/FN_HermiteSpline(
    Position Blend:NiagaraFloat=$Particles.Initial.PartitionParticles.Normalized Index In Partition
    Position End:NiagaraPosition=$User.Stop Location
    Position Start:NiagaraPosition=$Engine.Emitter.SimulationPosition
    Write Particles Position=true
   )
   /SetVariables(
    Particles.Beam Fade_Width:NiagaraFloat=&Multiply_Float(
     A=&Multiply_Float(A=$Particles.Initial.RibbonWidth, B=$Emitter.BeamWidthScale)
     B=&MaskFloatBySpawnGroup(Masked Float=3.5, Passthrough Float=0, Spawn Group Mask=2)
    )
    Particles.BeamDetail_Width:NiagaraFloat=&Multiply_Float(
     A=&Multiply_Float(A=$Particles.Initial.RibbonWidth, B=$Emitter.BeamWidthScale)
     B=&MaskFloatBySpawnGroup(
      Masked Float=/Game/Effects/Niagara/DynamicInputs/Math/FN_Fit(
       Clamp:NiagaraBool=false
       Value:NiagaraFloat=$Emitter.ScaleW
       SourceMin:NiagaraFloat=0
       SourceMax:NiagaraFloat=0.5
       DestMin:NiagaraFloat=3
       DestMax:NiagaraFloat=2.5
      )
      Passthrough Float=0
      Spawn Group Mask=4
     )
    )
    Particles.BeamGlow2_Width:NiagaraFloat=&Multiply_Float(
     A=&Multiply_Float(A=$Particles.Initial.RibbonWidth, B=$Emitter.BeamWidthScale)
     B=&MaskFloatBySpawnGroup(
      Masked Float=&FN_Fit(
       Clamp:NiagaraBool=false
       Value:NiagaraFloat=$Emitter.ScaleW
       SourceMin:NiagaraFloat=0
       SourceMax:NiagaraFloat=1
       DestMin:NiagaraFloat=0.5
       DestMax:NiagaraFloat=1
      )
      Passthrough Float=0
      Spawn Group Mask=3
     )
    )
    Particles.BeamGlow_Width:NiagaraFloat=&Multiply_Float(
     A=&Multiply_Float(A=$Particles.Initial.RibbonWidth, B=$Emitter.BeamWidthScale)
     B=&MaskFloatBySpawnGroup(Masked Float=0.8, Passthrough Float=0, Spawn Group Mask=1)
    )
    Particles.BeamParticle_Width:NiagaraFloat=&Multiply_Float(
     A=&Multiply_Float(A=$Particles.Initial.RibbonWidth, B=$Emitter.BeamWidthScale)
     B=&MaskFloatBySpawnGroup(Masked Float=1, Passthrough Float=0, Spawn Group Mask=5)
    )
    Particles.ProjectileBeam_Width:NiagaraFloat=&Multiply_Float(
     A=&Multiply_Float(A=$Particles.Initial.RibbonWidth, B=$Emitter.BeamWidthScale)
     B=&MaskFloatBySpawnGroup(Masked Float=1, Passthrough Float=0)
    )
   ) {isSetParameters=true}
   &ScaleColor(
    Scale Alpha=&MergeHelper_FloatSwitch(
     0=1
     1=0.4
     2=1
     3=&FloatFromCurve(
      FloatCurve=DataInterface() {propertyValues {Curve {keys [
          {interpMode=RCIM_Cubic, tangentMode=RCTM_Break, time=0, value=1}
          {arriveTangent=-3.624, interpMode=RCIM_Cubic, leaveTangent=-0.5922, tangentMode=RCTM_Break, time=0.7, value=0.3}
          {interpMode=RCIM_Cubic, tangentMode=RCTM_Break, time=1, value=0.4}
         ]}}}
      CurveIndex=$Particles.RibbonLinkOrder
     )
     4=1
     5=1
     Index=&SpawnGroup()
     Num=::NewEnumerator4
    )
    Scale RGB=<1 1 10>
    ScaleA=true
    ScaleRGB=true
   )
   &DynamicMaterialParameters(
    Index 0 Param 1=&MaskFloatBySpawnGroup(Masked Float=2, Passthrough Float=1)
    Index 2 Param 1=&Lerp_Float(
     A=-0.8775
     B=1
     Alpha=&Multiply_Float(
      A=&FloatFromCurve(FloatCurve=DataInterface() {propertyValues {Curve {keys [<at 0 0>,<at 0.1 1>]}}}, CurveIndex=$Engine.System.Age)
      B=1
     )
    )
    Index 3 Param 1=&FN_Fit(
     Clamp:NiagaraBool=false
     Value:NiagaraFloat=$Emitter.ScaleW
     SourceMin:NiagaraFloat=0
     SourceMax:NiagaraFloat=1
     DestMin:NiagaraFloat=0
     DestMax:NiagaraFloat=&FloatFromCurve(
      FloatCurve=DataInterface() {propertyValues {Curve {keys [{interpMode=RCIM_Cubic, tangentMode=RCTM_Break, time=0, value=0}, <at 0.5 1>]}}}
      CurveIndex=$Engine.System.Age
      Scale Curve=&FloatFromCurve(
       FloatCurve=DataInterface() {propertyValues {Curve {keys [{interpMode=RCIM_Cubic, leaveTangent=-3.7341, tangentMode=RCTM_Break, time=0, value=1}, <at 1 -1>]}}}
       CurveIndex=$System.TimeForDis
      )
     )
    )
    Module.Index0Param0Write=true
    Module.Index0Param1Write=true
    Module.Index0Param2Write=false
    Module.Index0Param3Write=false
    Module.Index2Param0Write=true
    Module.Index2Param1Write=true
    Module.Index2Param2Write=true
    Module.Index2Param3Write=true
    Module.Index3Param0Write=true
    Module.Index3Param1Write=true
    Module.Index3Param2Write=true
    Module.Index3Param3Write=true
    Write Parameter Index 2=true
    Write Parameter Index 3=true
   )
   &ScaleColor(
    Scale Alpha=&FloatFromCurve(
     FloatCurve=DataInterface() {propertyValues {Curve {
        defaultValue=1
        keys [
         {tangentMode=RCTM_User, time=0, value=1}
         {arriveTangent=-0.0405, interpMode=RCIM_Cubic, leaveTangent=-0.0405, tangentMode=RCTM_User, time=0.7009, value=1}
         {arriveTangent=-8.0089, interpMode=RCIM_Cubic, leaveTangent=-8.0089, tangentMode=RCTM_User, time=1, value=0}
        ]
       }}}
     CurveIndex=$Emitter.NormalizedLoopAge
    )
    Scale RGB=<50 10 300>
    ScaleA=true
    ScaleRGB=true
   ) {module=ScaleColor001}
   &ScaleRibbonWidth(Ribbon Width Scale=&FloatFromCurve(FloatCurve=DataInterface() {propertyValues {Curve {keys [<at 0 1>,<at 1 0>]}}}, CurveIndex=$Particles.PartitionParticles.NormalizedPartition)) {enabled=false}
  ]
  rendererClasses [&NiagaraRibbonRendererProperties]
  renderers [
   {rendererClass=&NiagaraRibbonRendererProperties}
   {rendererClass=&NiagaraRibbonRendererProperties, rendererIndex=1}
   {rendererClass=&NiagaraRibbonRendererProperties, rendererIndex=2}
   {rendererClass=&NiagaraRibbonRendererProperties, rendererIndex=3}
  ]
 }
 CenterFlare001 {
  simTarget=GPUComputeSim
  emitterUpdateScript [
   &EmitterState(Life Cycle Mode=::NewEnumerator1, Loop Behavior=::NewEnumerator1, Loop Duration=$User.Duration, UseLoopDelay=false)
   &SpawnBurst_Instantaneous(Use Spawn Probability=false, UseLoopCountLimit=false)
  ]
  particleSpawnScript [&InitializeParticle(
    Color=<rgba 1.5408 1.3217 2 1>
    Lifetime=$User.Duration
    Sprite Size=&Multiply_Vector2D_ByFloat(Float=$User.BeamBaseScale, Vector2D=<2400 1400>)
    Sprite Size Mode=::NewEnumerator4
    UsePositionOffset=false
    Write Material Random=true
   )]
  particleUpdateScript [
   &ParticleState(Kill Particles When Lifetime Has Elapsed=false, Let Infinitely Lived Particles Die When Emitter Deactivates=true)
   &self:MaskInScreenSpace_02() {module=MaskInScreenSpace_02}
   &self:ScratchModule_04(NumberOfRings=3, OcclusionSize=300, SamplesPerRing=8) {module=ScratchModule_04}
   &ScaleColor(
    Curve Index=$Transient.DistanceFromScreenEdge
    Linear Color Curve=DataInterface() {propertyValues {
      AlphaCurve {keys [<at 0 1>]}
      BlueCurve {keys [<at 0.0295 0.3803>,<at 0.8973 2>]}
      GreenCurve {keys [<at 0.0295 0>,<at 0.8973 1.6561>]}
      RedCurve {keys [<at 0.0295 0.5>,<at 0.8973 0.1108>]}
     }}
    Scale Mode=::NewEnumerator2
   ) {enabled=false, module=ScaleColor002}
   /SetVariables(Particles.Position=$Engine.Owner.Position) {enabled=false, isSetParameters=true, module=SetVariables_#13}
   &ScaleColor(
    Scale Alpha=$Particles.Visibility
    Scale RGB=&MakeVectorFromLinearColorRGB(Color=&ColorFromCurve(
      ColorCurve=DataInterface() {propertyValues {
        AlphaCurve {keys [<at 0 1 RCIM_Cubic>]}
        BlueCurve {keys [<at 0 2.5 RCIM_Cubic>,<at 1 2.3743 RCIM_Cubic>]}
        GreenCurve {keys [<at 0 1.482 RCIM_Cubic>,<at 1 0 RCIM_Cubic>]}
        RedCurve {keys [<at 0 0 RCIM_Cubic>,<at 1 2.5 RCIM_Cubic>]}
       }}
      CurveIndex=&Waveform("Waveform [1]"=::NewEnumerator0, "[1] Frequency"=2, "[1] Amplitude Min/Max"=<0 1>)
     ))
    ScaleA=true
    ScaleRGB=true
   )
   &ScaleSpriteSize(Uniform Curve Sprite Scale=DataInterface() {propertyValues {Curve {keys [{tangentMode=RCTM_User, time=0.95, value=1}, {tangentMode=RCTM_User, time=1, value=0}]}}})
   &Color(Color=&ColorFromCurve(
     ColorCurve=DataInterface() {propertyValues {
       AlphaCurve {keys [<at 0 1 RCIM_Cubic>]}
       BlueCurve {keys [<at 0 2.5 RCIM_Cubic>,<at 1 2.3743 RCIM_Cubic>]}
       GreenCurve {keys [<at 0 1.482 RCIM_Cubic>,<at 1 0 RCIM_Cubic>]}
       RedCurve {keys [<at 0 0 RCIM_Cubic>,<at 1 2.5 RCIM_Cubic>]}
      }}
     CurveIndex=&Waveform("Waveform [1]"=::NewEnumerator0, "[1] Frequency"=2, "[1] Amplitude Min/Max"=<0 1>)
    )) {enabled=false}
   &ScaleColor(
    Color Value To Scale=$Particles.Color
    Scale Alpha=&Multiply_Float(A=$User.GlobalOpacity, B=1)
    Scale RGB=&Multiply_Vector(A=&MakeVectorFromLinearColorRGB(Color=$Particles.AtmosphereColor), B=<1 1 1>)
    ScaleA=true
    ScaleRGB=true
   ) {enabled=false, module=ScaleColor001}
  ]
  rendererClasses [&NiagaraSpriteRendererProperties]
  renderers [{rendererClass=&NiagaraSpriteRendererProperties}]
 }
 CenterFlareLow {
  simTarget=GPUComputeSim
  emitterUpdateScript [
   &EmitterState(Life Cycle Mode=::NewEnumerator1, Loop Behavior=::NewEnumerator1, Loop Duration=$User.Duration, UseLoopDelay=false)
   &SpawnBurst_Instantaneous(Use Spawn Probability=false, UseLoopCountLimit=false)
  ]
  particleSpawnScript [&InitializeParticle(
    Color=<rgba 1.5408 1.3217 2 1>
    Lifetime=$User.Duration
    Sprite Size=&Multiply_Vector2D_ByFloat(Float=$User.BeamBaseScale, Vector2D=<1400 1000>)
    Sprite Size Mode=::NewEnumerator4
    UsePositionOffset=false
    Write Material Random=true
   )]
  particleUpdateScript [
   &ParticleState(Kill Particles When Lifetime Has Elapsed=false, Let Infinitely Lived Particles Die When Emitter Deactivates=true)
   &self:MaskInScreenSpace_02() {enabled=false, module=MaskInScreenSpace_02}
   &self:ScratchModule_04(NumberOfRings=3, OcclusionSize=300, SamplesPerRing=8) {enabled=false, module=ScratchModule_04}
   &ScaleColor(
    Curve Index=$Transient.DistanceFromScreenEdge
    Linear Color Curve=DataInterface() {propertyValues {
      AlphaCurve {keys [<at 0 1>]}
      BlueCurve {keys [<at 0.0295 0.3803>,<at 0.8973 2>]}
      GreenCurve {keys [<at 0.0295 0>,<at 0.8973 1.6561>]}
      RedCurve {keys [<at 0.0295 0.5>,<at 0.8973 0.1108>]}
     }}
    Scale Mode=::NewEnumerator2
   ) {enabled=false, module=ScaleColor002}
   /SetVariables(Particles.Position=$Engine.Owner.Position) {enabled=false, isSetParameters=true, module=SetVariables_#13}
   &ScaleColor(
    Scale Alpha=$Particles.Visibility
    Scale RGB=&MakeVectorFromLinearColorRGB(Color=&ColorFromCurve(
      ColorCurve=DataInterface() {propertyValues {
        AlphaCurve {keys [<at 0 1 RCIM_Cubic>]}
        BlueCurve {keys [<at 0 2.5 RCIM_Cubic>,<at 1 2.3743 RCIM_Cubic>]}
        GreenCurve {keys [<at 0 1.482 RCIM_Cubic>,<at 1 0 RCIM_Cubic>]}
        RedCurve {keys [<at 0 0 RCIM_Cubic>,<at 1 2.5 RCIM_Cubic>]}
       }}
      CurveIndex=&Waveform("Waveform [1]"=::NewEnumerator0, "[1] Frequency"=2, "[1] Amplitude Min/Max"=<0 1>)
     ))
    ScaleA=false
    ScaleRGB=true
   )
   &ScaleSpriteSize(Uniform Curve Sprite Scale=DataInterface() {propertyValues {Curve {keys [{tangentMode=RCTM_User, time=0.95, value=1}, {tangentMode=RCTM_User, time=1, value=0}]}}})
   &Color(Color=&ColorFromCurve(
     ColorCurve=DataInterface() {propertyValues {
       AlphaCurve {keys [<at 0 1 RCIM_Cubic>]}
       BlueCurve {keys [<at 0 2.5 RCIM_Cubic>,<at 1 2.3743 RCIM_Cubic>]}
       GreenCurve {keys [<at 0 1.482 RCIM_Cubic>,<at 1 0 RCIM_Cubic>]}
       RedCurve {keys [<at 0 0 RCIM_Cubic>,<at 1 2.5 RCIM_Cubic>]}
      }}
     CurveIndex=&Waveform("Waveform [1]"=::NewEnumerator0, "[1] Frequency"=2, "[1] Amplitude Min/Max"=<0 1>)
    )) {enabled=false}
   &ScaleColor(
    Color Value To Scale=$Particles.Color
    Scale Alpha=&Multiply_Float(A=$User.GlobalOpacity, B=1)
    Scale RGB=&Multiply_Vector(A=&MakeVectorFromLinearColorRGB(Color=$Particles.AtmosphereColor), B=<1 1 1>)
    ScaleA=true
    ScaleRGB=true
   ) {enabled=false, module=ScaleColor001}
   &CameraOffset(Camera Offset Amount=10)
  ]
  rendererClasses [&NiagaraSpriteRendererProperties]
  renderers [{rendererClass=&NiagaraSpriteRendererProperties}]
 }
 EndParticle {
  emitterUpdateScript [
   &EmitterState(Life Cycle Mode=::NewEnumerator1, Loop Behavior=::NewEnumerator1, Loop Duration=$User.Duration, UseLoopDelay=false)
   &SpawnPerFrame(Spawn:NiagaraBool=$System.DoOnce.Execute, Spawn Count=30, Use Spawn Probability=false)
   &SpawnBurst_Instantaneous(
    Spawn Count=30
    Spawn Time=&MakeCustomFloatFromBool(Bool=$User.bHitReaction, True Float=0, False Float=&Subtract_Float(A=$User.Duration, B=$User.DisappearTime))
    Use Spawn Probability=false
    UseLoopCountLimit=false
   )
  ]
  particleSpawnScript [
   &InitializeParticle(
    Color=<rgba 1.5834 0.5877 10 1>
    Lifetime=&RandomRangeFloat(Minimum=1, Maximum=1.5)
    Position Mode=::NewEnumerator1
    Sprite Size=&RandomRangeVector2D(
     Minimum=&Multiply_Vector2D_ByFloat(Float=$User.BeamWidth, Vector2D=<5 50>)
     Maximum=&Multiply_Vector2D_ByFloat(Float=$User.BeamWidth, Vector2D=<7 150>)
    )
    Sprite Size Mode=::NewEnumerator4
    UsePositionOffset=false
    Write Material Random=true
   )
   &ShapeLocation(
    Cylinder Height=500
    Cylinder Height Midpoint=0
    Cylinder Radius=50
    MaskSpawnGroup=false
    Override Seed=false
    Shape Primitive=::NewEnumerator1
    Surface Only=false
    UseOwnerScale=true
   )
   &AddVelocity(Velocity=&RandomRangeVector(Minimum=<0 0 127.4498>, Maximum=<0 0 159.1143>))
  ]
  particleUpdateScript [
   &ParticleState()
   &CurlNoiseForce(Bias Noise=false, Noise Strength=1000, Pan Noise=false)
   &Drag(Drag=&RandomRangeFloat(Minimum=0.5, Maximum=1), Use Linear Drag=true, Use Rotational Drag=true)
   &SolveForcesAndVelocity(Clamp Velocity=false, Limit Acceleration=false)
   &self:OrientToAxis_01(
    Quat:Quat4f=&MakeQuaternion(
     Quaternion From=3
     Axis=<1 0 0>
     Angle=0
     From=<0 0 -1>
     To=&NormalizeVector(
      Vector To Normalize=&Subtract_Position(A=<0 0 0>, B=Linked() {linkedVariable {name=System.Stop Location Local, type=Vector3f}})
      Safe Normalize=true
      Safe Fallback Threshold=0
      Safe Fallback Vector=<0 0 0.001>
     )
     Angle Type=::NewEnumerator0
     Normalize Vectors=false
    )
    ScaleXY:NiagaraFloat=&FloatFromCurve(FloatCurve=DataInterface() {propertyValues {Curve {keys [<at 0 1>,<at 2 0>]}}}, CurveIndex=$Particles.Age)
    ZLength:NiagaraFloat=&Divide_Float(
     A=&DistanceBetweenPositions(A=$Engine.Emitter.SimulationPosition, B=Linked() {linkedVariable {name=System.Stop Location Local, type=Vector3f}})
     B=500
    )
   ) {module=OrientToAxis_01}
   &FN_Sprite_Attributes(
    Color=&MultiplyLinearColor(Linear Color=$Particles.Color, Scale Factor=<0.025 0.025 0.025 0.5>)
    Create Color Attribute=true
    Create SpriteSize Attribute=true
    SpriteSize=&Multiply_Vector2D(A=$Particles.SpriteSize, B=<10.7658 2>)
    Write to Module Namespace=true
   )
   &ScaleColor(
    Scale Alpha=&FloatFromCurve(FloatCurve=DataInterface() {propertyValues {Curve {
        defaultValue=1
        keys [
         {arriveTangent=9.0725, interpMode=RCIM_Cubic, leaveTangent=9.0725, tangentMode=RCTM_User, time=0, value=0}
         {arriveTangent=-0.3964, interpMode=RCIM_Cubic, leaveTangent=-0.3964, tangentMode=RCTM_User, time=0.1408, value=0.9847}
         {arriveTangent=-0.0405, interpMode=RCIM_Cubic, leaveTangent=-0.0405, tangentMode=RCTM_User, time=1, value=0}
        ]
       }}})
    ScaleA=true
    ScaleRGB=true
   )
  ]
  rendererClasses [&NiagaraSpriteRendererProperties]
  renderers [
   {rendererClass=&NiagaraSpriteRendererProperties}
   {rendererClass=&NiagaraSpriteRendererProperties, rendererIndex=1}
  ]
 }
 Energy_Offset {
  emitterUpdateScript [
   &EmitterState(
    Life Cycle Mode=::NewEnumerator1
    Loop Behavior=::NewEnumerator1
    Loop Delay=0.1
    Loop Duration=&Subtract_Float(A=$User.Duration, B=1)
    UseLoopDelay=false
   )
   &SpawnBurst_Instantaneous(Spawn Count=2, Use Spawn Probability=false, UseLoopCountLimit=false) {enabled=false}
   &SpawnRate(SpawnRate=2, Use Spawn Probability=false)
   /SetVariables(Emitter.Linear Color:LinearColor=&LerpLinearColors(
     StartColor=<rgba 0.0159 0.1401 1 1>
     EndColor=<rgba 0.7387 0.1312 1 1>
     LerpFactor=&Waveform("Waveform [1]"=::NewEnumerator0, "[1] Frequency"=2, "[1] Amplitude Min/Max"=<0 1>)
    )) {isSetParameters=true}
  ]
  particleSpawnScript [
   &InitializeParticle(
    Mesh Scale=&Multiply_VectorByFloat(Vector=<0.5 0.5 1>, Float=$User.BeamBaseScale)
    Mesh Scale Mode=::NewEnumerator4
    UsePositionOffset=true
    Write Material Random=true
   )
   &InitialMeshOrientation(Mesh Axis To Orient=<0 0 1>, Mesh Orientation Mode=::NewEnumerator2, Use Rotation Vector=true)
   /SetVariables(Particles.MeshOrientation=&MultiplyQuaternion(
     Quaternion A=$Particles.InitialRotation
     Quaternion B=&MakeQuatFromAxisAngle(Axis=<0 0 1>, AngleInDegrees=&RandomRangeFloat(Minimum=-360, Maximum=360), NormalizeAxis=true)
    )) {isSetParameters=true}
  ]
  particleUpdateScript [
   &ParticleState()
   &ScaleMeshSize(Scale Factor=&VectorFromFloat(Value=&FloatFromCurve(FloatCurve=DataInterface() {propertyValues {Curve {keys [<at 0 0.5>,<at 0.1 1.2>]}}}, Scale Curve=$User.Charge)))
   &DynamicMaterialParameters(
    Index 0 Param 1=&FloatFromCurve(FloatCurve=DataInterface() {propertyValues {Curve {keys [<at 0 0.5>,<at 1 -0.63>]}}})
    Index 0 Param 2=&RandomRangeFloat(Minimum=-2, Maximum=1)
    Index 0 Param 3=0.1
    Module.Index0Param0Write=true
    Module.Index0Param1Write=true
    Module.Index0Param2Write=true
    Module.Index0Param3Write=false
   ) {module=DynamicMaterialParameters001}
   &ScaleColor(
    Scale Alpha=&FloatFromCurve(
     FloatCurve=DataInterface() {propertyValues {Curve {keys [<at 0 0>,<at 0.3 1>,<at 1 0>]}}}
     Scale Curve=&FloatFromCurve(FloatCurve=DataInterface() {propertyValues {Curve {keys [<at 0 0.5>,<at 0.2 1>]}}})
    )
    Scale RGB=&VectorFromFloat(Value=&FloatFromCurve(FloatCurve=DataInterface() {propertyValues {Curve {keys [<at 0 0>,<at 0.3 1>,<at 1 0>]}}}))
    ScaleA=true
    ScaleRGB=true
   )
   &UpdateMeshOrientation(Rotation Rate=0.025, Rotation Vector=<0 0 1>)
   &CameraOffset(Camera Offset Amount=350)
  ]
  rendererClasses [&NiagaraMeshRendererProperties]
  renderers [{rendererClass=&NiagaraMeshRendererProperties, rendererIndex=0}]
 }
 Impact_Flare {
  emitterUpdateScript [
   &EmitterState(Life Cycle Mode=::NewEnumerator1, Loop Behavior=::NewEnumerator1, Loop Duration=1, UseLoopDelay=false)
   &SpawnBurst_Instantaneous(Use Spawn Probability=false, UseLoopCountLimit=false)
  ]
  particleSpawnScript [
   &InitializeParticle(
    Color=<rgba 2 2 2 1>
    Lifetime=0.7
    Position Offset Coordinate Space=::Local
    Sprite Size Mode=::NewEnumerator3
    Uniform Sprite Size=&Multiply_Float(A=4500, B=$User.BeamBaseScale)
    UsePositionOffset=true
    Write Material Random=true
   )
   &DynamicMaterialParameters(
    Index 0 Param 1=1
    Index 0 Param 2=1
    Index 0 Param 3=1
    Index 0 Param 4=1
    Module.Param0WriteEnabled=true
    Module.Param1WriteEnabled=true
    Module.Param2WriteEnabled=true
    Module.Param3WriteEnabled=true
   )
   &CameraOffset(Camera Offset Amount=300)
  ]
  particleUpdateScript [
   &ParticleState()
   &ScaleColor(
    Scale Alpha=&FloatFromCurve(FloatCurve=DataInterface() {propertyValues {Curve {keys [
         {arriveTangent=6.0888, interpMode=RCIM_Cubic, leaveTangent=6.0888, tangentMode=RCTM_User, time=0, value=0}
         <at 0.4 1 RCIM_Cubic>
         <at 1 0 RCIM_Cubic>
        ]}}})
    Scale RGB=&MakeVectorFromLinearColorRGB(Color=&ColorFromCurve(ColorCurve=DataInterface() {propertyValues {
        AlphaCurve {keys []}
        BlueCurve {keys [<at 0 10 RCIM_Cubic>,<at 0.2799 1>,<at 1 0.1878 RCIM_Cubic>]}
        GreenCurve {keys [<at 0 3.7508 RCIM_Cubic>,<at 0.2799 0.2585>,<at 1 0.0093 RCIM_Cubic>]}
        RedCurve {keys [<at 0 0 RCIM_Cubic>,<at 0.2799 0.0694>,<at 1 0.0475 RCIM_Cubic>]}
       }}))
    ScaleA=true
    ScaleRGB=true
   )
   &ScaleColor(ScaleA=true, ScaleRGB=true) {module=ScaleColor001}
   &FN_Sprite_Attributes(
    Color=&Multiply_LinearColorByFloat(Linear Color=$Particles.Color, Float=2)
    Create Color Attribute=true
    Create SpriteSize Attribute=true
    SpriteSize=&Multiply_Vector2D(A=$Particles.SpriteSize, B=<1 0.4232>)
    Write to Module Namespace=true
   )
  ]
  rendererClasses [&NiagaraSpriteRendererProperties]
  renderers [
   {rendererClass=&NiagaraSpriteRendererProperties}
   {rendererClass=&NiagaraSpriteRendererProperties, rendererIndex=1}
  ]
 }
 Muzzle_Inner_Flares {
  emitterUpdateScript [
   &EmitterState()
   &SpawnBurst_Instantaneous(Use Spawn Probability=false, UseLoopCountLimit=false)
   &SpawnRate(SpawnRate=15, Use Spawn Probability=false)
  ]
  particleSpawnScript [
   &InitializeParticle(
    Color=$User.PrimaryColor
    Lifetime Max=0.175
    Lifetime Min=0.075
    Lifetime Mode=::NewEnumerator1
    Sprite Size Mode=::NewEnumerator1
    Uniform Sprite Size Max=&Multiply_Float(A=1000, B=$User.BeamBaseScale)
    Uniform Sprite Size Min=&Multiply_Float(A=600, B=$User.BeamBaseScale)
    UsePositionOffset=false
    Write Material Random=true
   )
   &CameraOffset(Camera Offset Amount=400)
   /SetVariables(
    Particles.Animorphic Color:LinearColor=$User.PrimaryColor
    Particles.Animorphic Size:Vector2f=&Multiply_Vector2D_ByFloat(Float=$User.BeamBaseScale, Vector2D=<1200 20>)
    Particles.Soft Glow Size:Vector2f=<3289.3108 195.8096>
   ) {isSetParameters=true}
  ]
  particleUpdateScript [
   &ParticleState()
   &ScaleColor(
    Scale Alpha=&FloatFromCurve(FloatCurve=DataInterface() {propertyValues {Curve {keys [
         {interpMode=RCIM_Cubic, tangentMode=RCTM_User, time=0, value=0}
         {interpMode=RCIM_Cubic, tangentMode=RCTM_User, time=0.5, value=1}
         {tangentMode=RCTM_User, time=1, value=0}
        ]}}})
    ScaleA=true
    ScaleRGB=true
   )
   /SetVariables(Particles.Soft Glow Color:LinearColor=&MakeLinearColorFromVectorAndFloat(
     "Vector (RGB)"=&MakeVectorFromLinearColorRGB(Color=<rgba 0.3559 0.3141 10 1>)
     "Float (Alpha)"=&FloatFromCurve(
      FloatCurve=DataInterface() {propertyValues {Curve {keys [
          {interpMode=RCIM_Cubic, tangentMode=RCTM_User, time=0, value=0}
          {interpMode=RCIM_Cubic, tangentMode=RCTM_User, time=0.5, value=1}
          {tangentMode=RCTM_User, time=1, value=0}
         ]}}}
      Scale Curve=&Multiply_Float(
       A=0.005
       B=&FloatFromCurve(
        FloatCurve=DataInterface() {propertyValues {Curve {keys [
            {interpMode=RCIM_Cubic, leaveTangent=-2.0566, tangentMode=RCTM_Break, time=0, value=1}
            {arriveTangent=-0.9453, tangentMode=RCTM_User, time=1, value=0}
           ]}}}
        CurveIndex=$System.TimeForDis
       )
      )
     )
    )) {isSetParameters=true}
   /SetVariables(
    Particles.Corona Color:LinearColor=&MakeLinearColorFromVectorAndFloat(
     "Vector (RGB)"=&MakeVectorFromLinearColorRGB(Color=<rgba 1 0.0889 0.1786 1>)
     "Float (Alpha)"=&FloatFromCurve(
      FloatCurve=DataInterface() {propertyValues {Curve {keys [
          <at 0 0 RCIM_Cubic>
          {arriveTangent=2.5, interpMode=RCIM_Cubic, leaveTangent=2.5, time=0.165, value=0.65}
          {arriveTangent=0.1786, interpMode=RCIM_Cubic, leaveTangent=0.1786, time=0.4, value=1}
          {arriveTangent=-1.6667, interpMode=RCIM_Cubic, leaveTangent=-1.6667, time=0.725, value=0.75}
          <at 1 0 RCIM_Cubic>
         ]}}}
      Scale Curve=&Multiply_Float(
       A=0.5
       B=&FloatFromCurve(
        FloatCurve=DataInterface() {propertyValues {Curve {keys [
            {interpMode=RCIM_Cubic, leaveTangent=-2.0566, tangentMode=RCTM_Break, time=0, value=1}
            {arriveTangent=-0.9453, tangentMode=RCTM_User, time=1, value=0}
           ]}}}
        CurveIndex=$System.TimeForDis
       )
      )
     )
    )
    Particles.Corona Size:Vector2f=&Vector2DFromFloat(Value=&FloatFromCurve(
      FloatCurve=DataInterface() {propertyValues {Curve {keys [<at 0 1>,<at 0.5 0.7>,<at 1 1>]}}}
      Scale Curve=&Waveform(Global Amplitude Scale=2000, "Waveform [1]"=::NewEnumerator12, "[1] Phase Scale"=50, "[1] Frequency"=1, "[1] Amplitude Min/Max"=<0.85 1>)
     ))
   ) {isSetParameters=true}
   &DynamicMaterialParameters(
    Index 3 Param 1=&FloatFromCurve(FloatCurve=DataInterface() {propertyValues {Curve {keys [<at 0 0.7 RCIM_Cubic>,<at 0.5 0.25 RCIM_Cubic>,<at 1 0.7 RCIM_Cubic>]}}})
    Index 3 Param 2=1
    Index 3 Param 3=&RandomRangeFloat(Minimum=0.25, Maximum=0.35)
    Index 3 Param 4=1
    Module.Param0WriteEnabled3=true
    Module.Param1WriteEnabled3=true
    Module.Param2WriteEnabled3=true
    Module.Param3WriteEnabled3=true
    Write Parameter Index 0=false
    Write Parameter Index 3=true
   )
   /SetVariables() {isSetParameters=true}
   &ScaleColor(
    Scale Alpha=&FloatFromCurve(
     FloatCurve=DataInterface() {propertyValues {Curve {keys [
         {interpMode=RCIM_Cubic, leaveTangent=-2.0566, tangentMode=RCTM_Break, time=0, value=1}
         {arriveTangent=-0.9453, tangentMode=RCTM_User, time=1, value=0}
        ]}}}
     CurveIndex=$System.TimeForDis
    )
    ScaleA=true
    ScaleRGB=true
   ) {module=ScaleColor001}
   /SetVariables(Particles.Animorphic Color:LinearColor=&MultiplyLinearColor(
     Linear Color=$Particles.Initial.Animorphic Color
     Scale Factor=&MakeVector4FromVectorAndFloat(
      "Vector (RGB)"=<1 1 1>
      "Float (Alpha)"=&FloatFromCurve(
       FloatCurve=DataInterface() {propertyValues {Curve {keys [{interpMode=RCIM_Cubic, leaveTangent=-3.7201, tangentMode=RCTM_Break, time=0, value=1}, <at 1 0>]}}}
       CurveIndex=$System.TimeForDis
       Scale Curve=0.1
      )
     )
    )) {isSetParameters=true}
  ]
  rendererClasses [&NiagaraSpriteRendererProperties]
  renderers [
   {rendererClass=&NiagaraSpriteRendererProperties}
   {rendererClass=&NiagaraSpriteRendererProperties, rendererIndex=1}
   {rendererClass=&NiagaraSpriteRendererProperties, rendererIndex=2}
   {rendererClass=&NiagaraSpriteRendererProperties, rendererIndex=3}
  ]
 }
 ShellBurst {
  emitterUpdateScript [
   &EmitterState()
   &SpawnBurst_Instantaneous(Use Spawn Probability=false, UseLoopCountLimit=false)
  ]
  particleSpawnScript [
   &InitializeParticle(
    Color=<rgba 1 1 1 0.5>
    Lifetime=$User.Duration
    Mesh Scale Mode=::NewEnumerator3
    Mesh Uniform Scale=26.2282
    Position=<15 0 5>
    Position Mode=::NewEnumerator1
    Sprite Size=&Multiply_Vector2D_ByFloat(Float=$User.BeamBaseScale, Vector2D=<700 500>)
    Sprite Size Mode=::NewEnumerator4
    UsePositionOffset=false
    Write Material Random=true
   )
   &CameraOffset(Camera Offset Amount=300)
  ]
  particleUpdateScript [
   &ParticleState()
   &ScaleMeshSize(Scale Factor=&VectorFromFloat(Value=&FloatFromCurve(FloatCurve=DataInterface() {propertyValues {Curve {keys [<at 0 0.8>,<at 1 1>]}}}, Scale Curve=10)))
   &DynamicMaterialParameters(
    Index 0 Param 1=0.25
    Index 0 Param 2=-1
    Index 0 Param 3=1
    Index 0 Param 4=1
    Module.Index0Float0Write=true
    Module.Index0Float1Write=true
    Module.Index0Float2Write=true
    Module.Index0Float3Write=true
   )
   &ScaleColor(
    Scale Alpha=&FloatFromCurve(FloatCurve=DataInterface() {propertyValues {Curve {keys [<at 0.0999 1.0037 RCIM_Cubic>,<at 1 1>]}}})
    ScaleA=true
    ScaleRGB=true
   )
   /SetVariables(Particles.Position=&AddVectorToPosition(
     Position=&Lerp_Vector(A=$System.SampleSkeletalMesh.SampledPosition, B=$System.SampleSkeletalMesh001.SampledPosition, Alpha=0)
     Vector=&TransformVector(Vector=&Multiply_VectorByFloat(Vector=<20 0 10>, Float=1), Destination Space=::Simulation)
    )) {enabled=false, isSetParameters=true, module=SetVariables_#14}
   &SpriteFacingAndAlignment(
    Alignment Coordinate Space=::World
    Sprite Alignment=&NormalizeVector(
     Vector To Normalize=&SubtractVectorFromPosition(Position=$Engine.Owner.Position, Vector=Linked() {linkedVariable {name=User.Stop Location, type=NiagaraPosition}})
     Safe Normalize=true
     Safe Fallback Threshold=0
     Safe Fallback Vector=<0 0 0.001>
    )
    WriteAlignment=true
    WriteFacing=false
   )
   &ScaleColor(
    Scale Alpha=&FloatFromCurve(
     FloatCurve=DataInterface() {propertyValues {Curve {keys [{interpMode=RCIM_Cubic, tangentMode=RCTM_Break, time=0, value=0}, <at 0.3 1>, <at 0.9 1>, <at 1 0>]}}}
     CurveIndex=$Engine.System.Age
    )
    ScaleA=true
    ScaleRGB=true
   ) {enabled=false, module=ScaleColor001}
   &ScaleColor(
    Scale Alpha=&FloatFromCurve(
     FloatCurve=DataInterface() {propertyValues {Curve {keys [
         {interpMode=RCIM_Cubic, leaveTangent=-0.2526, tangentMode=RCTM_Break, time=0, value=1}
         {arriveTangent=-1.8299, tangentMode=RCTM_User, time=1, value=0}
        ]}}}
     CurveIndex=$System.TimeForDis
    )
    ScaleA=true
    ScaleRGB=true
   ) {module=ScaleColor002}
  ]
  rendererClasses [&NiagaraSpriteRendererProperties]
  renderers [{rendererClass=&NiagaraSpriteRendererProperties}]
 }
 ShellBurst001 {
  emitterUpdateScript [
   &EmitterState()
   &SpawnBurst_Instantaneous(Use Spawn Probability=false, UseLoopCountLimit=false)
  ]
  particleSpawnScript [
   &InitializeParticle(
    Lifetime=$User.Duration
    Mesh Scale Mode=::NewEnumerator3
    Mesh Uniform Scale=26.2282
    Position=<15 0 5>
    Position Mode=::NewEnumerator1
    Sprite Size=&Multiply_Vector2D_ByFloat(Float=$User.BeamBaseScale, Vector2D=<300 400>)
    Sprite Size Mode=::NewEnumerator4
    UsePositionOffset=false
    Write Material Random=true
   )
   &CameraOffset(Camera Offset Amount=300)
  ]
  particleUpdateScript [
   &ParticleState()
   &ScaleSpriteSize(
    Scale Factor=&Vector2DFromCurve(
     Vector2Curve=DataInterface() {propertyValues {XCurve {keys [<at 0 0>,<at 0.05 1>,<at 0.8 1>,<at 1 0>]}, YCurve {keys [<at 0 1>,<at 1 1>]}}}
     ScaleCurve=<1 2>
    )
    Scale Sprite Size Mode=::NewEnumerator2
   )
   &ScaleMeshSize(Scale Factor=&VectorFromFloat(Value=&FloatFromCurve(FloatCurve=DataInterface() {propertyValues {Curve {keys [<at 0 0>,<at 1 0.2>]}}})))
   &DynamicMaterialParameters(
    Index 0 Param 1=0.25
    Index 0 Param 2=-1
    Index 0 Param 3=1
    Index 0 Param 4=1
    Module.Index0Param0Write=true
    Module.Index0Param1Write=true
    Module.Index0Param2Write=true
    Module.Index0Param3Write=true
   )
   &ScaleColor(ScaleA=true, ScaleRGB=true)
   /SetVariables(Particles.Position=&AddVectorToPosition(
     Position=&Lerp_Vector(A=$System.SampleSkeletalMesh.SampledPosition, B=$System.SampleSkeletalMesh001.SampledPosition, Alpha=0)
     Vector=&TransformVector(Vector=&Multiply_VectorByFloat(Vector=<20 0 10>, Float=1), Destination Space=::Simulation)
    )) {enabled=false, isSetParameters=true, module=SetVariables_#14}
   &SpriteFacingAndAlignment(
    Alignment Coordinate Space=::World
    Sprite Alignment=&NormalizeVector(
     Vector To Normalize=&SubtractVectorFromPosition(Position=$User.Stop Location, Vector=Linked() {linkedVariable {name=Engine.Owner.Position, type=NiagaraPosition}})
     Safe Normalize=true
     Safe Fallback Threshold=0
     Safe Fallback Vector=<0 0 0.001>
    )
    WriteAlignment=true
    WriteFacing=false
   )
   &ScaleColor(
    Scale Alpha=&FloatFromCurve(
     FloatCurve=DataInterface() {propertyValues {Curve {keys [{interpMode=RCIM_Cubic, tangentMode=RCTM_Break, time=0, value=0}, <at 0.3 1>]}}}
     CurveIndex=$Engine.System.Age
    )
    ScaleA=true
    ScaleRGB=true
   ) {module=ScaleColor001}
   &ScaleColor(
    Scale Alpha=&FloatFromCurve(
     FloatCurve=DataInterface() {propertyValues {Curve {keys [{interpMode=RCIM_Cubic, leaveTangent=-3.7201, tangentMode=RCTM_Break, time=0, value=1}, <at 1 0>]}}}
     CurveIndex=$System.TimeForDis
    )
    ScaleA=true
    ScaleRGB=true
   ) {enabled=false, module=ScaleColor002}
  ]
  rendererClasses [&NiagaraSpriteRendererProperties]
  renderers [{rendererClass=&NiagaraSpriteRendererProperties}]
 }
}
summary {
 systemName=NS_PC_EnergyBall_Charge_Beam
 emitters [
  {emitterName=Muzzle_Inner_Flares, rendererClasses [&NiagaraSpriteRendererProperties]}
  {emitterName=Beams, rendererClasses [&NiagaraRibbonRendererProperties]}
  {emitterName=ShellBurst, rendererClasses [&NiagaraSpriteRendererProperties]}
  {emitterName=ShellBurst001, rendererClasses [&NiagaraSpriteRendererProperties]}
  {emitterName=Impact_Flare, rendererClasses [&NiagaraSpriteRendererProperties]}
  {emitterName=EndParticle, rendererClasses [&NiagaraSpriteRendererProperties]}
  {emitterName=CenterFlare001, simTarget=GPUComputeSim, rendererClasses [&NiagaraSpriteRendererProperties]}
  {emitterName=Energy_Offset, rendererClasses [&NiagaraMeshRendererProperties]}
  {emitterName=CenterFlareLow, simTarget=GPUComputeSim, rendererClasses [&NiagaraSpriteRendererProperties]}
 ]
 userVariables [
  User.BeamBaseScale {type=NiagaraFloat, defaultValue=5, description=""}
  User.BeamImpactScale {type=NiagaraFloat, defaultValue=5, description=""}
  User.BeamWidth {type=NiagaraFloat, defaultValue=2, description=""}
  User.bHitReaction {type=NiagaraBool}
  User.bSpawnWater {type=NiagaraBool}
  User.Charge {type=NiagaraFloat, defaultValue=3, description=""}
  User.Delay {type=NiagaraFloat, defaultValue=0, description=""}
  User.DisappearTime {type=NiagaraFloat, defaultValue=0.8, description=""}
  User.Duration {type=NiagaraFloat, defaultValue=10, description=""}
  User.EnergyBeamSwipe {type=NiagaraFloat, defaultValue=0.456, description=""}
  User.GlobalOpacity {type=NiagaraFloat, defaultValue=1, description=""}
  User.GlobalScale {type=NiagaraFloat, defaultValue=0.75, description=""}
  User.ImpactActive {type=NiagaraFloat, defaultValue=0, description=""}
  User.ImpactOffset {type=NiagaraFloat, defaultValue=0, description=""}
  User.ImpactScale {type=NiagaraFloat, defaultValue=2, description=""}
  User.Initial Start Time {type=NiagaraFloat, defaultValue=0, description=""}
  User.InterpolateRateOfChange {type=NiagaraFloat, defaultValue=20, description=""}
  User.OverallOpacity {type=NiagaraFloat, defaultValue=1, description=""}
  User.PrimaryColor {type=LinearColor, description="", defaultValue {struct=LinearColor, value=<rgba 0.8321 4.0034 10 1>}}
  User.SecondaryColor {type=LinearColor, description="", defaultValue {struct=LinearColor, value=<rgba 0.3885 0 1 1>}}
  User.Size {type=NiagaraFloat, defaultValue=1, description=""}
  User.Stop Location {type=NiagaraPosition, description="", defaultValue {}}
 ]
}
