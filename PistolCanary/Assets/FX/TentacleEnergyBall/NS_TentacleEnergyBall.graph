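# Niagara system for a purple tentacle energy ball: a persistent glowing plasma sphere with lightning and flares.
# Looping mesh and sprite emitters layer the ball, radial energy, spiral, lightning, glows and waves, with light renderers on lightning and glow.
# User.Scale sizes most layers, User.Hitting doubles plasma brightness, and animated material parameters, waveforms and a system timer drive motion and pulsing.
# Some emitters, including ChargeUp and Rays001, are disabled.
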NiagaraSystem /PistolCanary/Assets/FX/TentacleEnergyBall/NS_TentacleEnergyBall
dependencies {
 usedDataInterfaces [
  &NiagaraDataInterfaceCurve
  &NiagaraDataInterfaceColorCurve
  &NiagaraDataInterfaceArrayColor
  &NiagaraDataInterfaceVector2DCurve
 ]
 usedDynamicInputs [
  &MakeVector4_FromLinearColor
  &Add_Float
  &Lerp_Vector
 ]
 usedModules [
  &SystemState
  &self:SystemSpawnScript.SystemScriptSource.SystemScriptGraph.NiagaraNodeAssignment_0.SetVariables_#1
  &self:SystemSpawnScript.SystemScriptSource.SystemScriptGraph.NiagaraNodeAssignment_1.SetVariables_#2
  &FN_Timer
  &self:SystemSpawnScript.SystemScriptSource.SystemScriptGraph.NiagaraNodeAssignment_2.SetVariables_#3
  &self:SystemSpawnScript.SystemScriptSource.SystemScriptGraph.NiagaraNodeAssignment_3.SetVariables_#4
  &self:SystemSpawnScript.SystemScriptSource.SystemScriptGraph.NiagaraNodeAssignment_4.SetVariables_#5
  &self:SystemSpawnScript.SystemScriptSource.SystemScriptGraph.NiagaraNodeAssignment_5.SetVariables_#6
  &self:SystemSpawnScript.SystemScriptSource.SystemScriptGraph.NiagaraNodeAssignment_6.SetVariables_#7
  &self:SystemSpawnScript.SystemScriptSource.SystemScriptGraph.NiagaraNodeAssignment_7.SetVariables_#8
 ]
}
emitters {
 BIg_Lightning {
  emitterUpdateScript [
   &EmitterState(Inactive Response=::NewEnumerator1, Life Cycle Mode=::NewEnumerator1, Loop Delay=1, Loop Duration=1, UseLoopDelay=false)
   &SpawnRate(SpawnRate=10, Use Spawn Probability=false)
  ]
  particleSpawnScript [
   &InitializeParticle(
    Lifetime Max=0.15
    Lifetime Min=0.1
    Lifetime Mode=::NewEnumerator1
    Mesh Scale=&RandomRangeVector(Minimum=<1 0.5 0.5>, Maximum=<1 1.5 1.5>)
    Mesh Scale Mode=::NewEnumerator4
    UsePositionOffset=false
    Write Material Random=true
   )
   &InitialMeshOrientation(
    Mesh Orientation Mode=::NewEnumerator2
    Rotation=&RandomRangeVector(Minimum=<-1 -1 -1>, Maximum=<1 1 1>)
    Use Rotation Vector=true
   ) {module=InitialMeshOrientation001}
   &ShapeLocation(
    Cylinder Height=0
    Cylinder Radius=200
    MaskSpawnGroup=false
    Override Seed=false
    Shape Primitive=::NewEnumerator1
    Surface Only=false
    UseOwnerScale=true
   )
   /SetVariables(Particles.Mesh int32:NiagaraInt32=&RandomRangeInt(Minimum=0, Maximum=0)) {isSetParameters=true, module=SetVariables_#9}
   &CameraOffset()
   &AddVelocity(
    Cone Angle=40
    Cone Axis=<-1 0 0>
    Module.UseConeFalloff=false
    Module.UseVelDistribution=false
    Override Seed=false
    Velocity Mode=::NewEnumerator2
    Velocity Speed=100
   )
  ]
  particleUpdateScript [
   &ParticleState()
   &ScaleMeshSize(Scale Factor=&Multiply_VectorByFloat(Vector=<1 1 1>, Float=&Multiply_Float(A=$User.Scale, B=100)))
   /SetVariables(Particles.Glow Color=&MakeLinearColorFromVectorAndFloat(
     "Vector (RGB)"=&Multiply_VectorByFloat(Vector=&MakeVectorFromLinearColorRGB(Color=<rgba 0.35 0 3 1>), Float=1)
     "Float (Alpha)"=&Multiply_Float(A=1, B=1)
    )) {isSetParameters=true, module=SetVariables_#10}
   /SetVariables(Particles.DY2:Vector4f=&MakeVector4(X=3, Y=5, Z=3, W=1)) {isSetParameters=true, module=SetVariables_#11}
   &DynamicMaterialParameters(
    Index 0 Param 1=&Multiply_Float(A=&RandomRangeFloat(Minimum=$User.Scale, Maximum=50), B=1)
    Index 0 Param 2=&Multiply_Float(A=3, B=$User.Scale)
    Index 0 Param 3=3
    Index 0 Param 4=1
    Module.Param0WriteEnabled=true
    Module.Param1WriteEnabled=true
    Module.Param2WriteEnabled=true
    Module.Param3WriteEnabled=true
   )
   &ScaleColor(
    Scale Alpha=&Multiply_Float(A=10, B=1)
    Scale RGB=&Multiply_VectorByFloat(
     Vector=<1 1 1>
     Float=&Multiply_Float(
      A=&FloatFromCurve(
       FloatCurve=DataInterface() {propertyValues {Curve {keys [
           <at 0 0.5>
           <at 0.3 0.5>
           {arriveTangent=-0.8, interpMode=RCIM_Cubic, leaveTangent=-0.8, time=0.5, value=0.1}
           {arriveTangent=0.8, interpMode=RCIM_Cubic, leaveTangent=0.8, time=0.8, value=0.1}
           <at 1 0.5>
          ]}}}
       Scale Curve=0.5
      )
      B=&Multiply_Float(A=1, B=1)
     )
    )
    ScaleA=true
    ScaleRGB=true
   )
   &SolveForcesAndVelocity(Clamp Velocity=false, Limit Acceleration=false)
   &FN_Light_Attributes(
    Color=<rgba 3.3333 0 10 1>
    Create Color Attribute=true
    Create Exponent Attribute=true
    Create Position Attribute=true
    Create Radius Attribute=true
    Exponent=1
    Position=&AddVectorToPosition(Position=$Particles.Position, Vector=<-500 0 -300>)
    Radius=500
    Scale Color=&FloatFromCurve(FloatCurve=DataInterface() {propertyValues {Curve {keys [
         {interpMode=RCIM_Cubic, tangentMode=RCTM_User, time=0, value=0}
         {interpMode=RCIM_Cubic, tangentMode=RCTM_User, time=0.3, value=1}
         <at 0.7 1>
         {tangentMode=RCTM_User, time=1, value=0}
        ]}}})
   )
  ]
  rendererClasses [&NiagaraMeshRendererProperties, &NiagaraLightRendererProperties]
  renderers [
   {rendererClass=&NiagaraMeshRendererProperties, rendererIndex=0}
   {rendererClass=&NiagaraMeshRendererProperties, rendererIndex=1}
   {rendererClass=&NiagaraLightRendererProperties, rendererIndex=2}
  ]
 }
 BigBall {
  emitterUpdateScript [
   &EmitterState(
    Inactive Response=::NewEnumerator1
    Life Cycle Mode=::NewEnumerator1
    Loop Behavior=::NewEnumerator1
    Loop Duration Mode=::NewEnumerator1
    UseLoopDelay=false
   )
   &SpawnBurst_Instantaneous(Spawn Count=1, Use Spawn Probability=false)
  ]
  particleSpawnScript [
   &InitializeParticle(
    Mesh Scale Mode=::NewEnumerator3
    Mesh Uniform Scale=0.015
    Sprite Size Mode=::NewEnumerator3
    Uniform Sprite Size=150
    UsePositionOffset=false
    Write Material Random=true
   )
   &InitialMeshOrientation(Orientation Vector=<0 0 1>, Rotation=<0 0.5 0>, Use Orientation Vector=true, Use Rotation Vector=true)
  ]
  particleUpdateScript [
   &ParticleState(Kill Particles When Lifetime Has Elapsed=false, Let Infinitely Lived Particles Die When Emitter Deactivates=true)
   &DynamicMaterialParameters(
    Index 0 Param 2=HlslExpression() {hlslExpression=Emitter.Age*.2}
    Index 0 Param 3=1
    Index 0 Param 4=1
    Index 1 Param 1=1
    Index 1 Param 2=1
    Index 1 Param 3=1
    Index 1 Param 4=1
    Index 2 Param 2=HlslExpression() {hlslExpression=Emitter.Age*.2}
    Index 2 Param 3=2
    Index 2 Param 4=2
    Module.Param0WriteEnabled=true
    Module.Param0WriteEnabled1=true
    Module.Param0WriteEnabled2=true
    Module.Param1WriteEnabled=true
    Module.Param1WriteEnabled1=true
    Module.Param1WriteEnabled2=true
    Module.Param2WriteEnabled=true
    Module.Param2WriteEnabled1=true
    Module.Param2WriteEnabled2=true
    Module.Param3WriteEnabled=true
    Module.Param3WriteEnabled1=true
    Module.Param3WriteEnabled2=true
    Write Parameter Index 1=true
    Write Parameter Index 2=true
   ) {module=DynamicMaterialParameters001}
   &ScaleMeshSize(Scale Factor=&Multiply_VectorByFloat(Vector=<1 1 1>, Float=400)) {module=ScaleMeshSize005}
   &ScaleMeshSize(Scale Factor=&Multiply_VectorByFloat(Vector=<1 1 1>, Float=&Multiply_Float(A=11, B=$User.Scale))) {module=ScaleMeshSize004}
   &SystemLocation(MaskSpawnGroup=false, Offset=$System.Center Position)
   &ScaleColor(Scale RGB=&Multiply_VectorByFloat(Vector=<1 1 1>, Float=1), ScaleA=true, ScaleRGB=true)
  ]
  rendererClasses [&NiagaraMeshRendererProperties]
  renderers [{rendererClass=&NiagaraMeshRendererProperties, rendererIndex=0}]
 }
 ChargeUp {
  bEnabled=false
  emitterUpdateScript [
   &EmitterState()
   &SpawnBurst_Instantaneous(Use Spawn Probability=false, UseLoopCountLimit=false)
  ]
  particleSpawnScript [
   &InitializeParticle(
    Lifetime=20
    Mesh Scale Mode=::NewEnumerator3
    Sprite Size=&Multiply_Vector2D_ByFloat(Float=&Multiply_Float(A=$User.ElementSizeScale, B=$User.PostSizeScale), Vector2D=<3948.9736 3201.8015>)
    Sprite Size Mode=::NewEnumerator4
    UsePositionOffset=false
    Write Material Random=true
   )
   &CameraOffset()
  ]
  particleUpdateScript [
   &ParticleState(Kill Particles When Lifetime Has Elapsed=false)
   &Timeline(
    Play Rate=&FloatFromCurve(
     FloatCurve=DataInterface() {propertyValues {Curve {keys [
         {interpMode=RCIM_Cubic, leaveTangent=-12.5334, tangentMode=RCTM_Break, time=0, value=4}
         {interpMode=RCIM_Cubic, tangentMode=RCTM_Break, time=0.7, value=1}
        ]}}}
     Scale Curve=0.692
    )
    Timeline End=1.0000000200408773e+20
   )
   &ScaleSpriteSize(Uniform Curve Sprite Scale=DataInterface() {propertyValues {Curve {keys [
        {arriveTangent=15.597, interpMode=RCIM_Cubic, leaveTangent=15.597, tangentMode=RCTM_User, time=0, value=0}
        {arriveTangent=0.3869, interpMode=RCIM_Cubic, leaveTangent=0.3869, tangentMode=RCTM_User, time=0.05, value=1}
        {arriveTangent=0.0224, interpMode=RCIM_Cubic, leaveTangent=0.0224, tangentMode=RCTM_User, time=1, value=1}
       ]}}})
   &DynamicMaterialParameters(
    Index 0 Param 1=&FloatFromCurve(
     FloatCurve=DataInterface() {propertyValues {Curve {keys [
         {interpMode=RCIM_Cubic, leaveTangent=-4.8343, tangentMode=RCTM_Break, time=0, value=2}
         {arriveTangent=-0.1577, interpMode=RCIM_Cubic, tangentMode=RCTM_Break, time=1, value=0}
        ]}}}
     Scale Curve=&Multiply_Float(A=2, B=$User.PostNoiseStrength)
    )
    Index 0 Param 2=&FloatFromCurve(
     FloatCurve=DataInterface() {propertyValues {Curve {keys [
         {interpMode=RCIM_Cubic, leaveTangent=-16.2213, tangentMode=RCTM_Break, time=0, value=4}
         {arriveTangent=-17.124, interpMode=RCIM_Cubic, tangentMode=RCTM_Break, time=0.25, value=0}
        ]}}}
     Scale Curve=&Multiply_Float(A=1, B=$User.PostStrength)
    )
    Index 0 Param 3=1
    Index 0 Param 4=1
    Index 1 Param 1=0.1
    Index 1 Param 2=1500
    Index 1 Param 3=0.5
    Index 1 Param 4=1
    Index 2 Param 1=$Particles.Timeline.Time
    Index 2 Param 2=$Particles.Timeline.Time
    Index 2 Param 3=$Particles.Timeline.Time
    Index 2 Param 4=&Multiply_Float(A=0.2, B=$User.ElementSizeScale)
    Index 3 Param 1=&FloatFromCurve(FloatCurve=DataInterface() {propertyValues {Curve {keys [<at 0.1 1>,<at 0.4 0>]}}})
    Index 3 Param 2=200
    Module.Index0Param0Write=true
    Module.Index0Param1Write=true
    Module.Index0Param2Write=true
    Module.Index0Param3Write=true
    Module.Index1Param0Write=true
    Module.Index1Param1Write=true
    Module.Index1Param2Write=true
    Module.Index1Param3Write=true
    Module.Index2Param0Write=true
    Module.Index2Param1Write=true
    Module.Index2Param2Write=true
    Module.Index2Param3Write=true
    Module.Index3Param0Write=true
    Module.Index3Param1Write=true
    Module.Index3Param2Write=true
    Module.Index3Param3Write=true
    Write Parameter Index 1=true
    Write Parameter Index 2=true
    Write Parameter Index 3=true
   )
   &ScaleColor(
    Scale Alpha=&FloatFromCurve(FloatCurve=DataInterface() {propertyValues {Curve {keys [{interpMode=RCIM_Cubic, leaveTangent=-6.3065, tangentMode=RCTM_Break, time=0.2, value=1}, <at 1 0>]}}})
    ScaleA=true
    ScaleRGB=true
   )
   /SetVariables(Particles.BlendTest:NiagaraFloat=&FloatFromCurve(
     FloatCurve=DataInterface() {propertyValues {Curve {keys [{arriveTangent=0.2621, interpMode=RCIM_Cubic, leaveTangent=0.2621, tangentMode=RCTM_User, time=0, value=1}]}}}
     Scale Curve=&Subtract_Float(
      A=$User.PostProcess
      B=&FN_LOD_Distance_Remap(Start Distance=8000, End Distance=120000, Minimum Output=0, Maximum Output=1, Clamp Output=true)
     )
    )) {isSetParameters=true}
   /SetVariables(Particles.ComponentsEnabled:NiagaraBool=&Boolean_NotOperation(Bool=&SetBoolByFloatComparison(
      A=&FN_LOD_Distance_Remap(Start Distance=4000, End Distance=6000, Minimum Output=0, Maximum Output=1, Clamp Output=true)
      B=1
     ))) {isSetParameters=true}
   &KillParticles(Kill Particles=&SetBoolByFloatComparison(
     A=&FloatFromCurve(FloatCurve=DataInterface() {propertyValues {Curve {keys [<at 0.9 1>,<at 1 -0.01>]}}})
     B=0
     Comparison Type=::NewEnumerator6
    ))
  ]
  rendererClasses [&NiagaraComponentRendererProperties]
  renderers [{rendererClass=&NiagaraComponentRendererProperties, rendererIndex=0}]
 }
 Flare_03 {
  emitterUpdateScript [
   &EmitterState(
    Inactive Response=::NewEnumerator1
    Life Cycle Mode=::NewEnumerator1
    Loop Behavior=::NewEnumerator1
    Loop Delay=0.133
    Loop Duration=9
    UseLoopDelay=false
   )
   &SpawnBurst_Instantaneous(Use Spawn Probability=false, UseLoopCountLimit=false)
   &SpawnRate(SpawnRate=3, Use Spawn Probability=false)
   &SpawnBurst_Instantaneous(Spawn Time=0.133, Use Spawn Probability=false, UseLoopCountLimit=false) {module=SpawnBurst_Instantaneous001}
  ]
  particleSpawnScript [
   &InitializeParticle(
    Color=&SelectLinearColorFromArray(
     Array Sampling Mode=::NewEnumerator1
     Direct Array Index=&RandomRangeInt(Minimum=0, Maximum=1)
     Color Selection Array=DataInterface() {propertyValues {ColorData [<rgba 0.3925 0 1 0.5>, <rgba 0 0.1777 1 0.5>]}}
     Random Seed=?
    )
    Lifetime Max=0.4
    Lifetime Min=0.2
    Lifetime Mode=::NewEnumerator1
    Sprite Size=&Multiply_Vector2D_ByFloat(
     Float=&Multiply_Float(A=1, B=&RandomRangeFloat(Minimum=1.2, Maximum=1.5))
     Vector2D=<250 180>
    )
    Sprite Size Mode=::NewEnumerator4
    UsePositionOffset=false
    Write Material Random=true
   )
   &SystemLocation(MaskSpawnGroup=false, Offset=<0 0 100>)
   &CameraOffset(Camera Offset Amount=-80)
  ]
  particleUpdateScript [
   &ParticleState()
   &ScaleColor(
    Scale Alpha=&Multiply_Float(
     A=&FloatFromCurve(
      FloatCurve=DataInterface() {propertyValues {Curve {keys [
          {interpMode=RCIM_Cubic, tangentMode=RCTM_User, time=0, value=0}
          {interpMode=RCIM_Cubic, tangentMode=RCTM_User, time=0.5, value=1}
          {tangentMode=RCTM_User, time=1, value=0}
         ]}}}
      Scale Curve=&FN_FrontEndFloatModifier(Gameplay=1, Front End=0.5)
     )
     B=&FloatFromCurve(
      FloatCurve=DataInterface() {propertyValues {Curve {keys [
          {arriveTangent=25.4717, interpMode=RCIM_Cubic, leaveTangent=25.4717, tangentMode=RCTM_User, time=0, value=0}
          {arriveTangent=0.1236, interpMode=RCIM_Cubic, leaveTangent=0.1236, tangentMode=RCTM_User, time=0.1, value=1}
          <at 0.5 1 RCIM_Cubic>
          {arriveTangent=-0.4595, interpMode=RCIM_Cubic, leaveTangent=-0.4595, tangentMode=RCTM_User, time=0.8, value=1}
          {arriveTangent=-26.3419, interpMode=RCIM_Cubic, leaveTangent=-26.3419, tangentMode=RCTM_User, time=0.9, value=0}
         ]}}}
      CurveIndex=$System.NormalizedLoopAge
     )
    )
    Scale RGB=&Multiply_VectorByFloat(
     Vector=&VectorFromFloat(Value=&FNTimeOfDayScalar(Night Scale=0.5, Morning Scale=0.25, Evening Scale=0.5))
     Float=&FN_FrontEndFloatModifier(Gameplay=1, Front End=1)
    )
    ScaleA=true
    ScaleRGB=true
   )
   &ScaleSpriteSize(
    Uniform Curve Scale=&Multiply_Float(A=$User.Scale, B=50)
    Uniform Curve Sprite Scale=DataInterface() {propertyValues {Curve {keys [{tangentMode=RCTM_User, time=0, value=0}, <at 0.1 0.7>, {tangentMode=RCTM_User, time=0.9, value=1}, <at 1 0>]}}}
   )
   &DynamicMaterialParameters(
    Index 0 Param 1=&RandomRangeFloat(Minimum=-1, Maximum=1)
    Index 0 Param 4=&RandomRangeFloat(Minimum=0, Maximum=1)
    Module.Param0WriteEnabled=true
    Module.Param1WriteEnabled=true
    Module.Param2WriteEnabled=true
    Module.Param3WriteEnabled=true
   )
  ]
  rendererClasses [&NiagaraSpriteRendererProperties]
  renderers [{rendererClass=&NiagaraSpriteRendererProperties}]
 }
 Glow {
  emitterUpdateScript [
   &EmitterState(
    Inactive Response=::NewEnumerator1
    Life Cycle Mode=::NewEnumerator1
    Loop Behavior=::NewEnumerator1
    Loop Duration Mode=::NewEnumerator1
    UseLoopDelay=false
   )
   &SpawnBurst_Instantaneous(Spawn Count=1, Use Spawn Probability=false)
  ]
  particleSpawnScript [
   &InitializeParticle(
    Lifetime=10
    Mesh Scale Mode=::NewEnumerator3
    Mesh Uniform Scale=0.015
    Sprite Size Mode=::NewEnumerator3
    Uniform Sprite Size=500
    UsePositionOffset=false
    Write Material Random=true
   )
   &InitialMeshOrientation(Orientation Vector=<0 0 1>, Rotation=<0 0.5 0>, Use Orientation Vector=true, Use Rotation Vector=true)
  ]
  particleUpdateScript [
   &ParticleState(Kill Particles When Lifetime Has Elapsed=false, Let Infinitely Lived Particles Die When Emitter Deactivates=true)
   &DynamicMaterialParameters(
    Index 0 Param 1=1
    Index 0 Param 2=1
    Index 0 Param 3=1
    Index 0 Param 4=1
    Index 1 Param 1=1
    Index 1 Param 2=1
    Index 1 Param 3=1
    Index 1 Param 4=1
    Index 2 Param 2=HlslExpression() {hlslExpression=Emitter.Age*.2}
    Index 2 Param 3=2
    Index 2 Param 4=2
    Module.Param0WriteEnabled=true
    Module.Param0WriteEnabled1=true
    Module.Param0WriteEnabled2=true
    Module.Param1WriteEnabled=true
    Module.Param1WriteEnabled1=true
    Module.Param1WriteEnabled2=true
    Module.Param2WriteEnabled=true
    Module.Param2WriteEnabled1=true
    Module.Param2WriteEnabled2=true
    Module.Param3WriteEnabled=true
    Module.Param3WriteEnabled1=true
    Module.Param3WriteEnabled2=true
    Write Parameter Index 1=true
    Write Parameter Index 2=true
   ) {module=DynamicMaterialParameters001}
   &ScaleSpriteSize(Scale Sprite Size Mode=::NewEnumerator0, Uniform Scale Factor=5) {module=ScaleSpriteSize005}
   &ScaleSpriteSize(Scale Sprite Size Mode=::NewEnumerator0, Uniform Scale Factor=&Multiply_Float(A=6, B=$User.Scale))
   /SetVariables(Particles.Glow Camera offset=100) {isSetParameters=true}
   /SetVariables(Particles.Glow Color=&MakeLinearColorFromVectorAndFloat(
     "Vector (RGB)"=&Multiply_VectorByFloat(Vector=&MakeVectorFromLinearColorRGB(Color=<rgba 0.0131 0 0.0521 1>), Float=3)
     "Float (Alpha)"=&Multiply_Float(A=1, B=1)
    )) {isSetParameters=true}
   &FN_Light_Attributes(
    Color=<rgba 0.7996 0 5 1>
    Create Color Attribute=true
    Create Exponent Attribute=true
    Create Position Attribute=true
    Create Radius Attribute=true
    Exponent=2
    Radius=&Multiply_Float(A=$User.Scale, B=10000)
    Scale Color=50
   )
   &SystemLocation(MaskSpawnGroup=false, Offset=$System.Center Position)
   /SetVariables(Particles.Size low:Vector2f=&Vector2DFromFloat(Value=&Multiply_Float(A=500, B=&Multiply_Float(A=6, B=&Multiply_Float(A=0.5, B=$System.Size Scale))))) {isSetParameters=true}
  ]
  rendererClasses [&NiagaraSpriteRendererProperties, &NiagaraLightRendererProperties]
  renderers [
   {rendererClass=&NiagaraSpriteRendererProperties}
   {rendererClass=&NiagaraLightRendererProperties, rendererIndex=1}
   {rendererClass=&NiagaraSpriteRendererProperties, rendererIndex=2}
  ]
 }
 Init_Glow {
  emitterUpdateScript [
   &EmitterState(Life Cycle Mode=::NewEnumerator1, Loop Behavior=::NewEnumerator1, Loop Delay=$System.TwoBallBurstTime, Loop Duration=1, UseLoopDelay=true)
   &SpawnBurst_Instantaneous(Spawn Count=1, Spawn Time=0.2, Use Spawn Probability=false)
  ]
  particleSpawnScript [
   &InitializeParticle(
    Color=<rgba 9.1752 0 10 5>
    Lifetime=0.7
    Mesh Scale Mode=::NewEnumerator3
    Mesh Uniform Scale=&Multiply_Float(A=1, B=3.5)
    UsePositionOffset=false
    Write Material Random=true
   ) {module=InitializeParticle001}
   &InitialMeshOrientation(
    Rotation=&RandomRangeVector(Minimum=<0 0 0>, Maximum=<1 1 1>)
    Use Orientation Vector=true
    Use Rotation Vector=true
   ) {module=InitialMeshOrientation001}
   &SystemLocation(MaskSpawnGroup=false) {module=SystemLocation001}
  ]
  particleUpdateScript [
   &ParticleState()
   &DynamicMaterialParameters(
    Index 0 Param 1=&FloatFromCurve(FloatCurve=DataInterface() {propertyValues {Curve {keys [<at 0 1.5>,<at 0.4 1>,<at 1 0.968>]}}})
    Index 0 Param 2=1
    Index 0 Param 3=1
    Index 0 Param 4=1
    Module.Param0WriteEnabled=true
    Module.Param1WriteEnabled=true
    Module.Param2WriteEnabled=true
    Module.Param3WriteEnabled=true
   )
   &ScaleColor(
    Scale Alpha=&FloatFromCurve(FloatCurve=DataInterface() {propertyValues {Curve {keys [<at 0 0.5 RCIM_Cubic>,<at 1 0>]}}})
    Scale RGB=&MakeVectorFromLinearColorRGB(Color=&ColorFromCurve(ColorCurve=DataInterface() {propertyValues {
        AlphaCurve {keys [<at 0 1 RCIM_Cubic>]}
        BlueCurve {keys [<at 0 1 RCIM_Cubic>,<at 1 1 RCIM_Cubic>]}
        GreenCurve {keys [<at 0 1 RCIM_Cubic>,<at 1 0.5246 RCIM_Cubic>]}
        RedCurve {keys [<at 0 1 RCIM_Cubic>,<at 1 0 RCIM_Cubic>]}
       }}))
    ScaleA=true
    ScaleRGB=true
   )
   &ScaleMeshSize(Scale Factor=&VectorFromFloat(Value=&FloatFromCurve(
      FloatCurve=DataInterface() {propertyValues {Curve {keys [{interpMode=RCIM_Cubic, leaveTangent=1.6767, tangentMode=RCTM_Break, time=0, value=0}, <at 1 1.7>]}}}
      Scale Curve=&Multiply_Float(A=1.2, B=1)
     )))
   &CameraOffset(Camera Offset Amount=50)
  ]
  rendererClasses [&NiagaraMeshRendererProperties]
  renderers [{rendererClass=&NiagaraMeshRendererProperties, rendererIndex=0}]
 }
 Plasma_Center_Energy {
  emitterUpdateScript [
   &EmitterState(
    Inactive Response=::NewEnumerator1
    Life Cycle Mode=::NewEnumerator1
    Loop Behavior=::NewEnumerator1
    Loop Duration Mode=::NewEnumerator1
    UseLoopDelay=false
   )
   &SpawnBurst_Instantaneous(Spawn Count=1, Use Spawn Probability=false)
  ]
  particleSpawnScript [&InitializeParticle(
    Color Mode=::NewEnumerator0
    Lifetime=$User.ChargeUpLength
    Sprite Size Mode=::NewEnumerator3
    Uniform Sprite Size=25
    UsePositionOffset=false
    Write Material Random=true
   )]
  particleUpdateScript [
   &ParticleState(Kill Particles When Lifetime Has Elapsed=false, Let Infinitely Lived Particles Die When Emitter Deactivates=true)
   &Color(
    Color=<rgba 5.6001 0 20 1>
    Scale Alpha=&FloatFromCurve(FloatCurve=DataInterface() {propertyValues {Curve {keys [<at 0 0.05>,<at 1 1>]}}})
    Scale Color=&VectorFromFloat(Value=&MakeCustomFloatFromBool(Bool=$User.Hitting, True Float=2, False Float=1))
   )
   &ScaleSpriteSize(Scale Sprite Size Mode=::NewEnumerator0, Uniform Scale Factor=200) {module=ScaleSpriteSize006}
   &ScaleSpriteSize(Scale Sprite Size Mode=::NewEnumerator0, Uniform Scale Factor=&Multiply_Float(A=4, B=$User.Scale)) {module=ScaleSpriteSize005}
   &SystemLocation(MaskSpawnGroup=false, Offset=$System.Center Position)
   &ScaleColor(
    Color Value To Scale=$Particles.Color
    Scale RGB=&Multiply_VectorByFloat(Vector=<1 1 1>, Float=$System.Tod)
    ScaleA=true
    ScaleRGB=true
   )
  ]
  rendererClasses [&NiagaraSpriteRendererProperties]
  renderers [{rendererClass=&NiagaraSpriteRendererProperties}]
 }
 Radial_Energy {
  emitterUpdateScript [
   &EmitterState(
    Inactive Response=::NewEnumerator1
    Life Cycle Mode=::NewEnumerator1
    Loop Behavior=::NewEnumerator1
    Loop Duration Mode=::NewEnumerator1
    UseLoopDelay=false
   )
   &SpawnBurst_Instantaneous(Spawn Count=1, Use Spawn Probability=false)
  ]
  particleSpawnScript [&InitializeParticle(
    Lifetime=$User.ChargeUpLength
    Sprite Size Mode=::NewEnumerator3
    Uniform Sprite Size=3.5
    UsePositionOffset=false
    Write Material Random=true
   )]
  particleUpdateScript [
   &ParticleState(Kill Particles When Lifetime Has Elapsed=false, Let Infinitely Lived Particles Die When Emitter Deactivates=true)
   &FN_Radial_Material(Base Sprite Tile V=2, Depth Fade=250, Radial Mask Power=10, Radial Mask Tile U=0, Sprite Extrusion Speed=0.6, Sprite Spin Speed=0.1)
   &Color(
    Color=&ColorFromCurve(
     ColorCurve=DataInterface() {propertyValues {
       AlphaCurve {keys [<at 0 1 RCIM_Cubic>,<at 1 1 RCIM_Cubic>]}
       BlueCurve {keys [<at 0 40 RCIM_Cubic>,<at 1 10 RCIM_Cubic>]}
       GreenCurve {keys [<at 0 0 RCIM_Cubic>,<at 1 5.4334 RCIM_Cubic>]}
       RedCurve {keys [<at 0 11 RCIM_Cubic>,<at 1 0 RCIM_Cubic>]}
      }}
     CurveIndex=&Waveform(
      Waveform Count=::NewEnumerator2
      "Waveform [1]"=::NewEnumerator0
      "[1] Frequency"=1
      "[1] Amplitude Min/Max"=<0 1>
      "[2] Average"=false
      "[2] Input Position"=0
      "[2] Phase"=$Emitter.Age
      "[2] Phase Scale"=1
      "[2] Compound Frequency"=2
      "[2] Frequency"=0.2
      "[2] Amplitude Min/Max"=<0 1>
      "[2] Scale"=1
      "[2] Bias"=0
      "[2] Exponent"=1
      "[2] Pulse Duration"=0.1
      "[2] Clip Triangle Min/Max"=<0 0>
      "[2] Random Seed"=0
      "[3] Average"=false
      "[3] Input Position"=0
      "[3] Phase"=$Emitter.Age
      "[3] Phase Scale"=1
      "[3] Frequency"=0.5
      "[3] Compound Frequency"=2
      "[3] Amplitude Min/Max"=<0 1>
      "[3] Scale"=1
      "[3] Bias"=0
      "[3] Exponent"=1
      "[3] Pulse Duration"=0.1
      "[3] Clip Triangle Min/Max"=<0 0>
      "[3] Random Seed"=0
     )
    )
    Scale Alpha=0.7
    Scale Color=&VectorFromFloat(Value=&MakeCustomFloatFromBool(Bool=$User.Hitting, True Float=2, False Float=1))
   )
   &ScaleSpriteSize(Scale Sprite Size Mode=::NewEnumerator0, Uniform Scale Factor=200) {module=ScaleSpriteSize005}
   &ScaleSpriteSize(Scale Sprite Size Mode=::NewEnumerator0, Uniform Scale Factor=&Multiply_Float(A=15, B=$User.Scale)) {module=ScaleSpriteSize004}
   &SystemLocation(MaskSpawnGroup=false, Offset=$System.Center Position)
   &ScaleColor(
    Color Value To Scale=$Particles.Color
    Scale RGB=&Multiply_VectorByFloat(Vector=<1 1 1>, Float=$System.Tod)
    ScaleA=true
    ScaleRGB=true
   )
  ]
  rendererClasses [&NiagaraSpriteRendererProperties]
  renderers [{rendererClass=&NiagaraSpriteRendererProperties}]
 }
 Radial_Plasma_Energy_02 {
  emitterUpdateScript [
   &EmitterState()
   &SpawnBurst_Instantaneous(Spawn Count=1, Use Spawn Probability=false)
  ]
  particleSpawnScript [&InitializeParticle(
    Color Mode=::NewEnumerator0
    Lifetime=10
    Sprite Size Mode=::NewEnumerator3
    Uniform Sprite Size=6
    UsePositionOffset=false
    Write Material Random=true
   )]
  particleUpdateScript [
   &ParticleState(Kill Particles When Lifetime Has Elapsed=false, Let Infinitely Lived Particles Die When Emitter Deactivates=true)
   &FN_Radial_Material(Base Sprite Tile V=2, Depth Fade=250, Radial Mask Power=10, Radial Mask Tile U=0, Sprite Extrusion Speed=0.6, Sprite Spin Speed=0.1)
   &Color(
    Color=<rgba 20 0 16.3794 1>
    Scale Alpha=0.7
    Scale Color=&VectorFromFloat(Value=&MakeCustomFloatFromBool(Bool=$User.Hitting, True Float=2, False Float=1))
   )
   &ScaleSpriteSize(Scale Factor=&Vector2DFromFloat(Value=$System.GrowingWOffset)) {enabled=false, module=ScaleSpriteSize001}
   &ScaleSpriteSize(Scale Factor=&Vector2DFromFloat(Value=$System.Shrinking)) {enabled=false, module=ScaleSpriteSize002}
   &ScaleSpriteSize(Scale Factor=&Vector2DFromFloat(Value=&Lerp_Float(
      A=1
      B=&Lerp_Float(
       A=0.8
       B=1
       Alpha=&Waveform("Waveform [1]"=::NewEnumerator0, "[1] Frequency"=7, "[1] Amplitude Min/Max"=<0 1>, "[2] Phase"=$Emitter.Age, "[3] Phase"=$Emitter.Age)
      )
      Alpha=$System.FN_Timer.Value
     ))) {enabled=false}
   &ScaleSpriteSize(
    Scale Sprite Size Mode=::NewEnumerator0
    Uniform Scale Factor=&Multiply_Float(
     A=&Multiply_Float(A=$User.Scale, B=3000)
     B=&Lerp_Float(
      A=1
      B=&Lerp_Float(
       A=0.8
       B=1
       Alpha=&Waveform("Waveform [1]"=::NewEnumerator0, "[1] Frequency"=7, "[1] Amplitude Min/Max"=<0 1>, "[2] Phase"=$Emitter.Age, "[3] Phase"=$Emitter.Age)
      )
      Alpha=$System.FN_Timer.Value
     )
    )
   ) {module=ScaleSpriteSize003}
  ]
  rendererClasses [&NiagaraSpriteRendererProperties]
  renderers [{rendererClass=&NiagaraSpriteRendererProperties}]
 }
 RainbowFlare001 {
  bEnabled=false
  emitterUpdateScript [
   &EmitterScalability(
    Module.bApplySpawnCountScale=false
    Module.bApplySpawnCountScaleByDistanceFraction=false
    Module.bUseMaxDistance=false
    Module.bUseMinDistance=false
   ) {enabled=false}
   &EmitterLifeCycle(MaxLoopCount=1, NextLoopDuration=1.7375) {enabled=false}
   &EmitterState(Inactive Response=::NewEnumerator1, Life Cycle Mode=::NewEnumerator1, Loop Behavior=::NewEnumerator1, Loop Duration=8.5, UseLoopDelay=false)
   &SpawnRate(Use Spawn Probability=false)
   &SpawnBurst_Instantaneous(Spawn Count=1, Use Spawn Probability=false) {enabled=false}
  ]
  particleSpawnScript [
   /Niagara/Modules/Spawn/Initialization/InitializeParticle(
    Lifetime=2
    Module.Write Color=true
    Module.Write Lifetime=true
    Module.Write Mass=true
    Module.Write Material Random=true
    Module.Write Position=true
    Module.Write Scale=false
    Module.Write SpriteRotation=true
    Module.Write SpriteSize=true
    Position=&SimulationPosition()
    Sprite Rotation=&UniformRangedFloat(Minimum=-360, Maximum=360)
    Sprite Size=&Vector2DFromFloat(Value=500)
   )
   &SystemLocation(MaskSpawnGroup=false, Offset=<0 0 100>)
   &CameraOffset(Camera Offset Amount=50)
  ]
  particleUpdateScript [
   &ParticleState()
   &SolveForcesAndVelocity(Clamp Velocity=false, Limit Acceleration=false)
   &ScaleColor(
    Scale Alpha=&FloatFromCurve(
     FloatCurve=DataInterface() {propertyValues {Curve {keys [
         {arriveTangent=6.4019, interpMode=RCIM_Cubic, leaveTangent=6.4019, tangentMode=RCTM_User, time=0, value=0}
         {arriveTangent=0.002, interpMode=RCIM_Cubic, leaveTangent=0.002, time=0.2, value=0.7}
         {arriveTangent=-0.9006, interpMode=RCIM_Cubic, leaveTangent=-0.9006, tangentMode=RCTM_User, time=1.0011, value=0.002}
        ]}}}
     Scale Curve=&FloatFromCurve(FloatCurve=DataInterface() {propertyValues {Curve {keys [
          {interpMode=RCIM_Cubic, tangentMode=RCTM_User, time=0, value=0}
          {interpMode=RCIM_Cubic, tangentMode=RCTM_User, time=0.5, value=1}
          {tangentMode=RCTM_User, time=1, value=0}
         ]}}})
    )
    ScaleA=true
    ScaleRGB=false
   )
   &ScaleSpriteSize(Scale Factor=&Vector2DFromCurve(
     Vector2Curve=DataInterface() {propertyValues {
       XCurve {keys [
         {arriveTangent=0.0114, interpMode=RCIM_Cubic, leaveTangent=0.0114, tangentMode=RCTM_User, time=-0.0107, value=0.2}
         {arriveTangent=1.1193, interpMode=RCIM_Cubic, leaveTangent=1.1193, tangentMode=RCTM_User, time=1.0179, value=0.7}
        ]}
       YCurve {keys [
         {arriveTangent=0.0114, interpMode=RCIM_Cubic, leaveTangent=0.0114, tangentMode=RCTM_User, time=-0.0269, value=0.2}
         {arriveTangent=1.1193, interpMode=RCIM_Cubic, leaveTangent=1.1193, tangentMode=RCTM_User, time=1.0179, value=0.7}
        ]}
      }}
     ScaleCurve=&Multiply_Vector2D_ByFloat(Float=&Multiply_Float(A=$User.Scale, B=70), Vector2D=<1 1>)
    ))
   &DynamicMaterialParameters(
    Index 0 Param 1=10
    Index 0 Param 2=1
    Index 0 Param 3=1
    Index 0 Param 4=1
    Index 1 Param 2=0.2
    Index 1 Param 3=2
    Index 1 Param 4=&FloatFromCurve(FloatCurve=DataInterface() {propertyValues {Curve {keys [
         {arriveTangent=4.1785, interpMode=RCIM_Cubic, leaveTangent=4.1785, tangentMode=RCTM_User, time=0, value=0}
         {arriveTangent=1.1179, interpMode=RCIM_Cubic, leaveTangent=1.1179, tangentMode=RCTM_User, time=0.2042, value=0.6228}
         {arriveTangent=0.0224, interpMode=RCIM_Cubic, leaveTangent=0.0224, tangentMode=RCTM_User, time=1, value=1}
        ]}}})
    Module.Param0WriteEnabled=true
    Module.Param0WriteEnabled1=true
    Module.Param1WriteEnabled=true
    Module.Param1WriteEnabled1=true
    Module.Param2WriteEnabled=true
    Module.Param2WriteEnabled1=true
    Module.Param3WriteEnabled=true
    Module.Param3WriteEnabled1=true
    Write Parameter Index 1=true
   )
  ]
  rendererClasses [&NiagaraSpriteRendererProperties]
  renderers [{rendererClass=&NiagaraSpriteRendererProperties}]
 }
 Rays001 {
  bEnabled=false
  emitterUpdateScript [
   &EmitterState()
   &SpawnRate(SpawnRate=$User.LightRays, Use Spawn Probability=false)
  ]
  particleSpawnScript [
   /Niagara/Modules/Spawn/Initialization/V2/InitializeParticle(Color=<rgba 0 3.3035 6 1>, Mesh Scale Mode=::NewEnumerator3, Mesh Uniform Scale=300, UsePositionOffset=false, Write Material Random=true)
   &DynamicMaterialParameters(
    Index 0 Param 1=&RandomRangeFloat(Minimum=0, Maximum=10)
    Index 0 Param 2=&RandomRangeFloat(Minimum=2.5, Maximum=4.5)
    Index 0 Param 3=&RandomRangeFloat(Minimum=0.8, Maximum=1.6)
    Module.Index0Param0Write=true
    Module.Index0Param1Write=true
    Module.Index0Param2Write=true
    Module.Index0Param3Write=false
   )
  ]
  particleUpdateScript [
   &ParticleState()
   &ScaleMeshSize(Scale Factor=&VectorFromFloat(Value=&FloatFromCurve(FloatCurve=DataInterface() {propertyValues {Curve {keys [{arriveTangent=34.3634, interpMode=RCIM_Cubic, time=0, value=0}, {interpMode=RCIM_Cubic, leaveTangent=-0.1625, time=1, value=1}]}}})))
   &ScaleColor(
    Scale Alpha=&FloatFromCurve(FloatCurve=DataInterface() {propertyValues {Curve {keys [
         <at 0 0 RCIM_Cubic>
         {arriveTangent=1.25, interpMode=RCIM_Cubic, leaveTangent=1.25, time=0.15, value=1}
         {arriveTangent=-1.1765, interpMode=RCIM_Cubic, leaveTangent=-1.1765, time=0.8, value=1}
         <at 1 0 RCIM_Cubic>
        ]}}})
    ScaleA=true
    ScaleRGB=true
   )
   &CameraOffset(Camera Offset Amount=5)
  ]
  rendererClasses [&NiagaraMeshRendererProperties]
  renderers [{rendererClass=&NiagaraMeshRendererProperties, rendererIndex=0}]
 }
 SmallLightning {
  emitterUpdateScript [
   &EmitterState(Inactive Response=::NewEnumerator1, Life Cycle Mode=::NewEnumerator1, Loop Delay=1, Loop Duration=1, UseLoopDelay=false)
   &SpawnRate(SpawnRate=13, Use Spawn Probability=false)
  ]
  particleSpawnScript [
   &InitializeParticle(
    Lifetime Max=0.15
    Lifetime Min=0.1
    Lifetime Mode=::NewEnumerator1
    Mesh Scale=&RandomRangeVector(Minimum=<1 0.5 0.5>, Maximum=<1.3 1.5 1.5>)
    Mesh Scale Mode=::NewEnumerator4
    UsePositionOffset=false
    Write Material Random=true
   )
   &InitialMeshOrientation(
    Mesh Orientation Mode=::NewEnumerator2
    Rotation=&RandomRangeVector(Minimum=<0 0 0>, Maximum=<1 1 0>)
    Use Rotation Vector=true
   ) {module=InitialMeshOrientation001}
   &ShapeLocation(
    Cylinder Height=50
    Cylinder Radius=30
    MaskSpawnGroup=false
    Override Seed=false
    Shape Primitive=::NewEnumerator1
    Surface Only=false
    UseOwnerScale=true
   )
   /SetVariables(Particles.Mesh int32:NiagaraInt32=&RandomRangeInt(Minimum=0, Maximum=1)) {isSetParameters=true, module=SetVariables_#9}
   &CameraOffset()
   &AddVelocity(
    Module.Use Falloff Distance=false
    Override Seed=false
    Velocity Mode=::NewEnumerator1
    Velocity Origin=$Particles.Position
    Velocity Speed=100
   )
  ]
  particleUpdateScript [
   &ParticleState()
   &ScaleMeshSize(Scale Factor=&Multiply_VectorByFloat(Vector=<1 1 1>, Float=&Multiply_Float(A=$User.Scale, B=40)))
   /SetVariables(Particles.Glow Color=&MakeLinearColorFromVectorAndFloat(
     "Vector (RGB)"=&Multiply_VectorByFloat(Vector=&MakeVectorFromLinearColorRGB(Color=<rgba 0.35 0 3 1>), Float=1)
     "Float (Alpha)"=&Multiply_Float(A=1, B=1)
    )) {isSetParameters=true, module=SetVariables_#10}
   /SetVariables(Particles.DY2:Vector4f=&MakeVector4(X=3, Y=2, Z=2, W=1)) {isSetParameters=true, module=SetVariables_#11}
   &DynamicMaterialParameters(
    Index 0 Param 1=&Multiply_Float(A=&RandomRangeFloat(Minimum=$User.Scale, Maximum=50), B=1)
    Index 0 Param 2=&RandomRangeFloat(Minimum=2, Maximum=2)
    Index 0 Param 3=3
    Index 0 Param 4=1
    Module.Param0WriteEnabled=true
    Module.Param1WriteEnabled=true
    Module.Param2WriteEnabled=true
    Module.Param3WriteEnabled=true
   )
   &ScaleColor(
    Scale Alpha=&Multiply_Float(A=10, B=1)
    Scale RGB=&Multiply_VectorByFloat(
     Vector=<1 1 1>
     Float=&Multiply_Float(
      A=&FloatFromCurve(
       FloatCurve=DataInterface() {propertyValues {Curve {keys [
           <at 0 0.5>
           <at 0.3 0.5>
           {arriveTangent=-0.8, interpMode=RCIM_Cubic, leaveTangent=-0.8, time=0.5, value=0.1}
           {arriveTangent=0.8, interpMode=RCIM_Cubic, leaveTangent=0.8, time=0.8, value=0.1}
           <at 1 0.5>
          ]}}}
       Scale Curve=0.5
      )
      B=&Multiply_Float(A=1, B=1)
     )
    )
    ScaleA=true
    ScaleRGB=true
   )
   &SolveForcesAndVelocity(Clamp Velocity=false, Limit Acceleration=false)
   &SystemLocation(MaskSpawnGroup=false, Offset=$System.Center Position)
  ]
  rendererClasses [&NiagaraMeshRendererProperties]
  renderers [
   {rendererClass=&NiagaraMeshRendererProperties, rendererIndex=0}
   {rendererClass=&NiagaraMeshRendererProperties, rendererIndex=1}
  ]
 }
 Spiral_Energy_big01 {
  emitterUpdateScript [
   &EmitterState(Life Cycle Mode=::NewEnumerator1, Loop Behavior=::NewEnumerator1, Loop Duration Mode=::NewEnumerator1, UseLoopDelay=false)
   &SpawnBurst_Instantaneous(Use Spawn Probability=false, UseLoopCountLimit=false)
  ]
  particleSpawnScript [
   &InitializeParticle(
    Color=<rgba 0.625 0 2.5 10>
    Mesh Scale=&Multiply_VectorByFloat(Vector=<1 1.5 1>, Float=&Multiply_Float(A=$User.Scale, B=150))
    Mesh Scale Mode=::NewEnumerator4
    Sprite Size=<200 200>
    Sprite Size Mode=::NewEnumerator4
    UsePositionOffset=false
    Write Material Random=true
   )
   &InitialMeshOrientation(Mesh Orientation Mode=::NewEnumerator2, Rotation=<0 0 0.25>, Use Rotation Vector=true)
   &SystemLocation(MaskSpawnGroup=false, Offset=<400 0 0>)
   &CameraOffset(Camera Offset Amount=30)
   &InitialMeshOrientation(Mesh Orientation Mode=::NewEnumerator2, Rotation=<0.25 0 0>, Use Rotation Vector=true) {module=InitialMeshOrientation001}
  ]
  particleUpdateScript [
   &ParticleState(Kill Particles When Lifetime Has Elapsed=false, Let Infinitely Lived Particles Die When Emitter Deactivates=true)
   &ScaleMeshSize(Scale Factor=&VectorFromFloat(Value=$User.Scale))
   &UpdateMeshOrientation(Rotation Rate=0, Rotation Vector=<0 1 0>)
   &ScaleColor(Scale RGB=&Multiply_VectorByFloat(Vector=<1 1 1>, Float=1), ScaleA=true, ScaleRGB=true)
   &DynamicMaterialParameters(
    Index 0 Param 1=1
    Index 0 Param 2=1.5
    Module.Param0WriteEnabled=true
    Module.Param1WriteEnabled=true
    Module.Param2WriteEnabled=true
    Module.Param3WriteEnabled=true
   )
   /SetVariables(Particles.Color02:LinearColor=<rgba 4 0 1.1971 1.5>) {enabled=false, isSetParameters=true}
   /SetVariables(Particles.Size02:Vector3f=&Multiply_VectorByFloat(Vector=<1 1 1>, Float=25)) {enabled=false, isSetParameters=true}
   /SetVariables(Particles.Position02:NiagaraPosition=&AddVectorToPosition(Position=$Particles.Position, Vector=<-100 0 0>)) {enabled=false, isSetParameters=true}
  ]
  rendererClasses [&NiagaraMeshRendererProperties]
  renderers [{rendererClass=&NiagaraMeshRendererProperties, rendererIndex=0}]
 }
 Wave02 {
  emitterUpdateScript [
   &EmitterState(Life Cycle Mode=::NewEnumerator1, Loop Behavior=::NewEnumerator1, Loop Duration=1, UseLoopDelay=true)
   &SpawnBurst_Instantaneous(Spawn Count=1, Spawn Time=0.25, Use Spawn Probability=false)
  ]
  particleSpawnScript [
   &InitializeParticle(
    Color=<rgba 5 0 2.5246 1>
    Lifetime=0.8
    Mesh Scale Mode=::NewEnumerator3
    Mesh Uniform Scale=35
    UsePositionOffset=false
    Write Material Random=true
   )
   &InitialMeshOrientation(Orientation Axis=<0 0 1>, Orientation Vector=<0 0 1>, Rotation=<0 0.25 0>, Use Orientation Vector=true, Use Rotation Vector=true)
   &InitialMeshOrientation(
    Orientation Axis=<0 0 0>
    Orientation Vector=<0 0 0>
    Rotation=&RandomRangeVector(Minimum=<0 0 0>, Maximum=<1 0 0>)
    Use Orientation Vector=false
    Use Rotation Vector=true
   ) {module=InitialMeshOrientation001}
   &CameraOffset(Camera Offset Amount=100)
  ]
  particleUpdateScript [
   &ParticleState()
   &ScaleColor(
    Linear Color Curve=DataInterface() {propertyValues {
      AlphaCurve {keys [<at 0 1>,<at 0.3576 1>,<at 1 0>]}
      BlueCurve {keys [<at 0 0.0414>,<at 0.1684 0.45>,<at 0.3654 1>,<at 1 1>]}
      GreenCurve {keys [<at 0 1>,<at 0.1684 1>,<at 0.3654 0.701>,<at 1 0.0574>]}
      RedCurve {keys [<at 0 0.2012>,<at 0.1684 0>,<at 0.3654 0.1029>,<at 1 0.0414>]}
     }}
    Scale Mode=::NewEnumerator2
   ) {enabled=false}
   &DynamicMaterialParameters(
    Index 0 Param 1=&FloatFromCurve(FloatCurve=DataInterface() {propertyValues {Curve {keys [{arriveTangent=-2.1015, interpMode=RCIM_Cubic, leaveTangent=-2.1015, tangentMode=RCTM_User, time=0, value=0.5}, <at 1 -0.15>]}}})
    Index 0 Param 2=&FloatFromCurve(FloatCurve=DataInterface() {propertyValues {Curve {keys [{arriveTangent=-0.3464, interpMode=RCIM_Cubic, leaveTangent=-0.3464, tangentMode=RCTM_User, time=0, value=1}, <at 1 0.9>]}}})
    Index 0 Param 3=&FloatFromCurve(FloatCurve=DataInterface() {propertyValues {Curve {keys [<at 0 0>,<at 0.15 1>,<at 1 1.2>]}}})
    Module.Param0WriteEnabled=true
    Module.Param1WriteEnabled=true
    Module.Param2WriteEnabled=true
    Module.Param3WriteEnabled=false
   )
   &ScaleColor(
    Scale Alpha=&FloatFromCurve(FloatCurve=DataInterface() {propertyValues {Curve {keys [<at 0 1>,<at 1 0>]}}})
    Scale RGB=&VectorFromFloat(Value=&FloatFromCurve(FloatCurve=DataInterface() {propertyValues {Curve {keys [<at 0 10>,<at 0.15 3>,<at 1 1>]}}}))
    ScaleA=true
    ScaleRGB=true
   ) {module=ScaleColor001}
   &ScaleMeshSize(Scale Factor=&VectorFromFloat(Value=&FloatFromCurve(
      FloatCurve=DataInterface() {propertyValues {Curve {keys [<at 0 0>,<at 0.1 1.4>,<at 1 2>]}}}
      Scale Curve=&Multiply_Float(A=1, B=$System.Size Scale)
     )))
  ]
  rendererClasses [&NiagaraMeshRendererProperties]
  renderers [{rendererClass=&NiagaraMeshRendererProperties, rendererIndex=0}]
 }
}
summary {
 systemName=NS_TentacleEnergyBall
 emitters [
  {emitterName=Plasma_Center_Energy, rendererClasses [&NiagaraSpriteRendererProperties]}
  {emitterName=BigBall, rendererClasses [&NiagaraMeshRendererProperties]}
  {emitterName=Radial_Energy, rendererClasses [&NiagaraSpriteRendererProperties]}
  {emitterName=Glow, rendererClasses [&NiagaraSpriteRendererProperties, &NiagaraLightRendererProperties]}
  {emitterName=Spiral_Energy_big01, rendererClasses [&NiagaraMeshRendererProperties]}
  {emitterName=BIg_Lightning, rendererClasses [&NiagaraMeshRendererProperties, &NiagaraLightRendererProperties]}
  {emitterName=SmallLightning, rendererClasses [&NiagaraMeshRendererProperties]}
  {emitterName=Radial_Plasma_Energy_02, rendererClasses [&NiagaraSpriteRendererProperties]}
  {bEnabled=false, emitterName=ChargeUp, rendererClasses [&NiagaraComponentRendererProperties]}
  {emitterName=Flare_03, rendererClasses [&NiagaraSpriteRendererProperties]}
  {bEnabled=false, emitterName=RainbowFlare001, rendererClasses [&NiagaraSpriteRendererProperties]}
  {bEnabled=false, emitterName=Rays001, rendererClasses [&NiagaraMeshRendererProperties]}
  {emitterName=Wave02, rendererClasses [&NiagaraMeshRendererProperties]}
  {emitterName=Init_Glow, rendererClasses [&NiagaraMeshRendererProperties]}
 ]
 userVariables [
  User.bIsFrontEnd {type=NiagaraBool}
  User.bIsFrontEndPreview {type=NiagaraBool}
  User.ChargeUpLength {type=NiagaraFloat, defaultValue=0, description=""}
  User.ElementSizeScale {type=NiagaraFloat, defaultValue=1, description=""}
  User.Expired {type=NiagaraBool, defaultValue=true}
  User.ExpireStart {type=NiagaraBool, defaultValue=true}
  User.FX_Sphere_Small_Size {type=NiagaraFloat, defaultValue=0, description=""}
  User.Hitting {type=NiagaraBool}
  User.LightRays {type=NiagaraFloat, defaultValue=0, description=""}
  User.PostNoiseStrength {type=NiagaraFloat, defaultValue=1, description=""}
  User.PostProcess {type=NiagaraFloat, defaultValue=1, description=""}
  User.PostSizeScale {type=NiagaraFloat, defaultValue=1, description=""}
  User.PostStrength {type=NiagaraFloat, defaultValue=1, description=""}
  User.Scale {type=NiagaraFloat, defaultValue=1.5, description=""}
 ]
}
