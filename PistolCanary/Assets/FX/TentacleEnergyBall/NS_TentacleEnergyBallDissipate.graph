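# Niagara one-shot burst effect for the tentacle energy ball dissipating in a purple energy explosion.
# Layers mesh shockwaves, dark and fresnel spheres, lightning and glows with curl-noise sparkles, a radial plasma sprite and a shrinking light.
# System Size Scale sizes most layers, TwoBallBurstTime delays two of them, and sparkles grow with camera distance.
# User Scale and Hitting drive plasma size and brightness; front-end preview with Return suppresses the shockwave and dark sphere.
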NiagaraSystem /PistolCanary/Assets/FX/TentacleEnergyBall/NS_TentacleEnergyBallDissipate
dependencies {
 usedDataInterfaces [&NiagaraDataInterfaceCurve, &NiagaraDataInterfaceColorCurve, &NiagaraDataInterfaceArrayFloat]
 usedDynamicInputs [&SelectFloatFromArray]
 usedModules [
  &SystemState
  &self:SystemSpawnScript.SystemScriptSource.SystemScriptGraph.NiagaraNodeAssignment_0.SetVariables_#1
 ]
}
emitters {
 BIg_Lightning {
  emitterUpdateScript [
   &EmitterState(Life Cycle Mode=::NewEnumerator1, Loop Behavior=::NewEnumerator1, Loop Delay=1, Loop Duration=0.1, UseLoopDelay=false)
   &SpawnBurst_Instantaneous(Spawn Count=15, Use Spawn Probability=false, UseLoopCountLimit=false)
   &SpawnRate(SpawnRate=50, Use Spawn Probability=false)
  ]
  particleSpawnScript [
   &InitializeParticle(
    Lifetime Max=0.4
    Lifetime Min=0.2
    Lifetime Mode=::NewEnumerator1
    Mesh Scale=&RandomRangeVector(Minimum=<1 1 1>, Maximum=<1.5 1.5 1.5>)
    Mesh Scale Mode=::NewEnumerator4
    UsePositionOffset=false
    Write Material Random=true
   )
   &InitialMeshOrientation(
    Mesh Orientation Mode=::NewEnumerator2
    Rotation=&RandomRangeVector(Minimum=<-1 0 0>, Maximum=<0 0 0.2>)
    Use Rotation Vector=true
   ) {module=InitialMeshOrientation001}
   &ShapeLocation(
    Cylinder Height=0
    Cylinder Radius=300
    MaskSpawnGroup=false
    Override Seed=false
    Shape Primitive=::NewEnumerator1
    Surface Only=false
    UseOwnerScale=true
   ) {enabled=false}
   /SetVariables(Particles.Mesh int32:NiagaraInt32=&RandomRangeInt(Minimum=0, Maximum=0)) {isSetParameters=true}
   &CameraOffset()
   &AddVelocity(
    Module.Use Falloff Distance=false
    Override Seed=false
    Velocity Mode=::NewEnumerator1
    Velocity Origin=$Particles.Position
    Velocity Speed=1
   )
  ]
  particleUpdateScript [
   &ParticleState()
   &ScaleMeshSize(Scale Factor=&Multiply_VectorByFloat(Vector=<1 1 1>, Float=&Multiply_Float(A=5, B=$System.Size Scale)))
   /SetVariables(Particles.Glow Color=&MakeLinearColorFromVectorAndFloat(
     "Vector (RGB)"=&Multiply_VectorByFloat(Vector=&MakeVectorFromLinearColorRGB(Color=<rgba 0.35 0 3 1>), Float=1)
     "Float (Alpha)"=&Multiply_Float(A=1, B=1)
    )) {isSetParameters=true}
   /SetVariables(Particles.DY2:Vector4f=&MakeVector4(X=5, Y=4, Z=3, W=1)) {isSetParameters=true}
   &DynamicMaterialParameters(
    Index 0 Param 1=&Multiply_Float(A=&RandomRangeFloat(Minimum=3, Maximum=5), B=1)
    Index 0 Param 2=4
    Index 0 Param 3=3
    Index 0 Param 4=1
    Module.Param0WriteEnabled=true
    Module.Param1WriteEnabled=true
    Module.Param2WriteEnabled=true
    Module.Param3WriteEnabled=true
   )
   &ScaleColor(
    Scale Alpha=&Multiply_Float(A=10, B=1)
    Scale RGB=&Multiply_VectorByFloat(
     Vector=<1 1 1>
     Float=&Multiply_Float(
      A=&FloatFromCurve(
       FloatCurve=DataInterface() {propertyValues {Curve {keys [
           <at 0 0.5>
           <at 0.3 0.5>
           {arriveTangent=-0.8, interpMode=RCIM_Cubic, leaveTangent=-0.8, time=0.5, value=0.1}
           {arriveTangent=0.8, interpMode=RCIM_Cubic, leaveTangent=0.8, time=0.8, value=0.1}
           <at 1 0.5>
          ]}}}
       Scale Curve=0.5
      )
      B=&Multiply_Float(A=1, B=1)
     )
    )
    ScaleA=true
    ScaleRGB=true
   )
   &SolveForcesAndVelocity(Clamp Velocity=false, Limit Acceleration=false)
  ]
  rendererClasses [&NiagaraMeshRendererProperties]
  renderers [
   {rendererClass=&NiagaraMeshRendererProperties, rendererIndex=0}
   {rendererClass=&NiagaraMeshRendererProperties, rendererIndex=1}
  ]
 }
 ExplodeFresnelSphere {
  emitterUpdateScript [
   /SetVariables(Emitter.JustSphere=1) {isSetParameters=true}
   &EmitterState(Life Cycle Mode=::NewEnumerator1, Loop Behavior=::NewEnumerator1, Loop Delay=$System.TwoBallBurstTime, Loop Duration=1, UseLoopDelay=true)
   &SpawnBurst_Instantaneous(Spawn Count=1, Use Spawn Probability=false)
  ]
  particleSpawnScript [
   &InitializeParticle(
    Color=<rgba 1.5205 0.4593 3 1>
    Lifetime=0.4
    Mesh Scale Mode=::NewEnumerator3
    Mesh Uniform Scale=14
    UsePositionOffset=false
    Write Material Random=true
   )
   &SystemLocation(MaskSpawnGroup=false) {module=SystemLocation001}
  ]
  particleUpdateScript [
   &ParticleState()
   &ScaleColor(
    Linear Color Curve=DataInterface() {propertyValues {
      AlphaCurve {keys [<at 0 1>]}
      BlueCurve {keys [<at 0 1>,<at 0.2382 1>,<at 1 1>]}
      GreenCurve {keys [<at 0 0.4865>,<at 0.2382 0.1138>,<at 1 0.1198>]}
      RedCurve {keys [<at 0 0.1198>,<at 0.2382 0.2941>,<at 1 0.2958>]}
     }}
    Scale Mode=::NewEnumerator2
   )
   &ScaleMeshSize(Scale Factor=&VectorFromFloat(Value=&FloatFromCurve(
      FloatCurve=DataInterface() {propertyValues {Curve {keys [
          {arriveTangent=3, interpMode=RCIM_Cubic, time=0, value=0.2}
          {arriveTangent=1.2, leaveTangent=1.2, time=0.2, value=1}
          {tangentMode=RCTM_User, time=1, value=1.2}
         ]}}}
      Scale Curve=&Multiply_Float(A=$System.Size Scale, B=1)
     )))
   &DynamicMaterialParameters(
    Index 0 Param 1=&FloatFromCurve(FloatCurve=DataInterface() {propertyValues {Curve {keys [<at 0 1>,<at 1 1.2>]}}})
    Index 0 Param 2=&FloatFromCurve(FloatCurve=DataInterface() {propertyValues {Curve {keys [<at 0 10>,<at 0.1 5>,<at 0.5 0>]}}})
    Module.Param0WriteEnabled=true
    Module.Param1WriteEnabled=true
    Module.Param2WriteEnabled=true
    Module.Param3WriteEnabled=true
   )
   &ScaleColor(
    Scale Alpha=&FloatFromCurve(FloatCurve=DataInterface() {propertyValues {Curve {keys [{arriveTangent=-2.7142, time=0, value=1}, <at 0.2 0.2>, <at 1 0>]}}})
    Scale RGB=&VectorFromFloat(Value=&FloatFromCurve(FloatCurve=DataInterface() {propertyValues {Curve {keys [<at 0 1>,<at 0.3 1>,<at 1 0>]}}}))
    ScaleA=true
    ScaleRGB=true
   ) {module=ScaleColor001}
  ]
  rendererClasses [&NiagaraMeshRendererProperties]
  renderers [{rendererClass=&NiagaraMeshRendererProperties, rendererIndex=0}]
 }
 Explode_Sparkles {
  emitterUpdateScript [
   &EmitterState(Life Cycle Mode=::NewEnumerator1, Loop Behavior=::NewEnumerator1, Loop Delay=0.1, Loop Duration=1, UseLoopDelay=false)
   &SpawnBurst_Instantaneous(Spawn Count=140, Use Spawn Probability=false, UseLoopCountLimit=false)
  ]
  particleSpawnScript [
   &InitializeParticle(
    Color=<rgba 3.3119 2.222 5 1>
    Lifetime=&RandomRangeFloat(Minimum=1, Maximum=3)
    Sprite Rotation Mode=::NewEnumerator0
    Sprite Size Max=<23.0769 230.7692>
    Sprite Size Min=<23.0769 153.8462>
    Sprite Size Mode=::NewEnumerator2
    UsePositionOffset=false
    Write Material Random=true
   )
   &SystemLocation(MaskSpawnGroup=false)
   &ShapeLocation(MaskSpawnGroup=false, Override Seed=false, Rotation Mode=::NewEnumerator1, Sphere Radius=300, UseOwnerScale=true)
   &AddVelocity(
    Module.Use Falloff Distance=false
    Override Seed=false
    Velocity Mode=::NewEnumerator1
    Velocity Speed=&RandomRangeFloat(Minimum=8400, Maximum=12000)
   )
  ]
  particleUpdateScript [
   /SetVariables(Particles.Glow03_Color=&MakeLinearColorFromVectorAndFloat(
     "Vector (RGB)"=&MakeVectorFromLinearColorRGB(Color=&MultiplyLinearColor(Linear Color=<rgba 0.2142 0 1 0.2>, Scale Factor=&MakeVector4_FromFloat(Float=1)))
     "Float (Alpha)"=&Multiply_Float(A=1, B=1)
    )) {isSetParameters=true}
   /SetVariables(Particles.Glow03_Size=&Multiply_Vector2D_ByFloat(
     Float=&FloatFromCurve(
      FloatCurve=DataInterface() {propertyValues {Curve {keys [
          <at 0 0 RCIM_Cubic>
          <at 0.2 1.3 RCIM_Cubic>
          {arriveTangent=-0.1423, interpMode=RCIM_Cubic, leaveTangent=-0.1423, tangentMode=RCTM_User, time=1, value=0}
         ]}}}
      Scale Curve=&Multiply_Float(
       A=10
       B=&ClampFloat(
        Float=&FN_LOD_Distance_Remap(Start Distance=1000, End Distance=10000, Minimum Output=1, Maximum Output=30)
        Max=2
       )
      )
     )
     Vector2D=<3 4>
    )) {isSetParameters=true}
   &ParticleState()
   &CurlNoiseForce(Noise Strength=&Multiply_Float(A=5000, B=1), Pan Noise=true, Pan Noise Field=<0 0 1>)
   &ScaleSpriteSize(Scale Factor=&Vector2DFromFloat(Value=&FloatFromCurve(
      FloatCurve=DataInterface() {propertyValues {Curve {keys [
          <at 0 0 RCIM_Cubic>
          <at 0.2 1.2 RCIM_Cubic>
          {arriveTangent=-2.7132, interpMode=RCIM_Cubic, leaveTangent=-2.7132, tangentMode=RCTM_User, time=1, value=0}
         ]}}}
      Scale Curve=&Multiply_Float(
       A=0.7
       B=&ClampFloat(
        Float=&FN_LOD_Distance_Remap(Start Distance=1000, End Distance=10000, Minimum Output=1, Maximum Output=20)
        Max=2
       )
      )
     )))
   &ScaleColor(
    Scale Alpha=&Multiply_Float(A=1, B=1)
    Scale RGB=&Multiply_VectorByFloat(Vector=<2 2 2>, Float=1)
    ScaleA=true
    ScaleRGB=true
   )
   &Drag(Use Linear Drag=true, Use Rotational Drag=false)
   &SolveForcesAndVelocity(Clamp Velocity=false, Limit Acceleration=false)
  ]
  rendererClasses [&NiagaraSpriteRendererProperties]
  renderers [
   {rendererClass=&NiagaraSpriteRendererProperties}
   {rendererClass=&NiagaraSpriteRendererProperties, rendererIndex=1}
  ]
 }
 Explode_SphereDark {
  emitterUpdateScript [
   &EmitterState(Life Cycle Mode=::NewEnumerator1, Loop Behavior=::NewEnumerator1, Loop Duration=1, UseLoopDelay=false)
   &SpawnBurst_Instantaneous(
    Spawn Count=&MakeCustomIntFromBool(
     Bool=$User.bIsFrontEndPreview
     True INT=&MakeCustomIntFromBool(Bool=$User.Return, True INT=0, False INT=1)
     False INT=1
    )
    Use Spawn Probability=false
    UseLoopCountLimit=false
   )
  ]
  particleSpawnScript [
   &InitializeParticle(
    Color=<rgba 0.0107 0 0.0625 3>
    Lifetime=0.6
    Mesh Scale Mode=::NewEnumerator3
    Mesh Uniform Scale=9
    UsePositionOffset=false
    Write Material Random=true
   )
   &InitialMeshOrientation(Mesh Orientation Mode=::NewEnumerator2, Rotation=<0 0.25 0>, Use Rotation Vector=true)
  ]
  particleUpdateScript [
   &ParticleState()
   &ScaleMeshSize(Scale Factor=&Multiply_VectorByFloat(
     Vector=<1 1 1>
     Float=&FloatFromCurve(
      FloatCurve=DataInterface() {propertyValues {Curve {keys [
          {interpMode=RCIM_Cubic, tangentMode=RCTM_User, time=0, value=0}
          <at 0.4 0.775>
          {interpMode=RCIM_Cubic, tangentMode=RCTM_User, time=1, value=1}
         ]}}}
      Scale Curve=&Multiply_Float(A=1.4, B=$System.Size Scale)
     )
    ))
   &DynamicMaterialParameters(
    Index 0 Param 1=&FloatFromCurve(FloatCurve=DataInterface() {propertyValues {Curve {keys [<at 0 0.8>,<at 1 1>]}}})
    Index 0 Param 2=&FloatFromCurve(FloatCurve=DataInterface() {propertyValues {Curve {keys [
         {interpMode=RCIM_Cubic, tangentMode=RCTM_User, time=0, value=1}
         {arriveTangent=-1, interpMode=RCIM_Cubic, leaveTangent=-1, time=0.6, value=0.7}
         {interpMode=RCIM_Cubic, leaveTangent=-3.3753, time=1, value=0}
        ]}}})
    Index 0 Param 3=1
    Index 0 Param 4=1
    Module.Param0WriteEnabled=true
    Module.Param1WriteEnabled=true
    Module.Param2WriteEnabled=true
    Module.Param3WriteEnabled=true
   )
   &ScaleColor(
    Scale Alpha=&FloatFromCurve(FloatCurve=DataInterface() {propertyValues {Curve {keys [
         {tangentMode=RCTM_User, time=0, value=1}
         {arriveTangent=-0.0405, interpMode=RCIM_Cubic, leaveTangent=-0.0405, tangentMode=RCTM_User, time=0.6151, value=1.0045}
         {arriveTangent=-4.906, interpMode=RCIM_Cubic, leaveTangent=-4.906, tangentMode=RCTM_User, time=1, value=0}
        ]}}})
    Scale RGB=&Multiply_VectorByFloat(Vector=<1 1 1>, Float=1)
    ScaleA=true
    ScaleRGB=true
   )
  ]
  rendererClasses [&NiagaraMeshRendererProperties]
  renderers [{rendererClass=&NiagaraMeshRendererProperties, rendererIndex=0}]
 }
 Glow {
  emitterUpdateScript [
   &EmitterState(Inactive Response=::NewEnumerator1, Life Cycle Mode=::NewEnumerator1, Loop Behavior=::NewEnumerator1, Loop Duration=1, UseLoopDelay=false)
   &SpawnBurst_Instantaneous(Spawn Count=1, Use Spawn Probability=false)
  ]
  particleSpawnScript [
   &InitializeParticle(
    Color Mode=::NewEnumerator0
    Mesh Scale Mode=::NewEnumerator3
    Mesh Uniform Scale=0.015
    Sprite Size Mode=::NewEnumerator3
    Uniform Sprite Size=500
    UsePositionOffset=false
    Write Material Random=true
   )
   &InitialMeshOrientation(Orientation Vector=<0 0 1>, Rotation=<0 0.5 0>, Use Orientation Vector=true, Use Rotation Vector=true)
  ]
  particleUpdateScript [
   &ParticleState()
   &DynamicMaterialParameters(
    Index 0 Param 1=1
    Index 0 Param 2=1
    Index 0 Param 3=1
    Index 0 Param 4=1
    Index 1 Param 1=1
    Index 1 Param 2=1
    Index 1 Param 3=1
    Index 1 Param 4=1
    Index 2 Param 2=HlslExpression() {hlslExpression=Emitter.Age*.2}
    Index 2 Param 3=2
    Index 2 Param 4=2
    Module.Param0WriteEnabled=true
    Module.Param0WriteEnabled1=true
    Module.Param0WriteEnabled2=true
    Module.Param1WriteEnabled=true
    Module.Param1WriteEnabled1=true
    Module.Param1WriteEnabled2=true
    Module.Param2WriteEnabled=true
    Module.Param2WriteEnabled1=true
    Module.Param2WriteEnabled2=true
    Module.Param3WriteEnabled=true
    Module.Param3WriteEnabled1=true
    Module.Param3WriteEnabled2=true
    Write Parameter Index 1=true
    Write Parameter Index 2=true
   ) {module=DynamicMaterialParameters001}
   &ScaleSpriteSize(Scale Sprite Size Mode=::NewEnumerator0, Uniform Scale Factor=5) {module=ScaleSpriteSize005}
   &ScaleSpriteSize(Scale Sprite Size Mode=::NewEnumerator0, Uniform Scale Factor=&Multiply_Float(A=1, B=$System.Size Scale))
   /SetVariables(Particles.Glow Camera offset=100) {isSetParameters=true}
   /SetVariables(Particles.Glow Color=&MakeLinearColorFromVectorAndFloat(
     "Vector (RGB)"=&Multiply_VectorByFloat(Vector=&MakeVectorFromLinearColorRGB(Color=<rgba 0.0521 0 0.0479 1>), Float=3)
     "Float (Alpha)"=&Multiply_Float(
      A=&FloatFromCurve(FloatCurve=DataInterface() {propertyValues {Curve {keys [<at 0 1>,<at 1 0>]}}}, Scale Curve=1.5)
      B=1
     )
    )) {isSetParameters=true}
  ]
  rendererClasses [&NiagaraSpriteRendererProperties]
  renderers [{rendererClass=&NiagaraSpriteRendererProperties}]
 }
 Init_Glow {
  emitterUpdateScript [
   &EmitterState(Life Cycle Mode=::NewEnumerator1, Loop Behavior=::NewEnumerator1, Loop Delay=$System.TwoBallBurstTime, Loop Duration=1, UseLoopDelay=true)
   &SpawnBurst_Instantaneous(Spawn Count=1, Use Spawn Probability=false)
  ]
  particleSpawnScript [
   &InitializeParticle(
    Color=<rgba 9.1752 0 10 5>
    Lifetime=0.7
    Mesh Scale Mode=::NewEnumerator3
    Mesh Uniform Scale=&Multiply_Float(A=$System.Size Scale, B=2)
    UsePositionOffset=false
    Write Material Random=true
   ) {module=InitializeParticle001}
   &InitialMeshOrientation(
    Rotation=&RandomRangeVector(Minimum=<0 0 0>, Maximum=<1 1 1>)
    Use Orientation Vector=true
    Use Rotation Vector=true
   ) {module=InitialMeshOrientation001}
   &SystemLocation(MaskSpawnGroup=false) {module=SystemLocation001}
  ]
  particleUpdateScript [
   &ParticleState()
   &DynamicMaterialParameters(
    Index 0 Param 1=&FloatFromCurve(FloatCurve=DataInterface() {propertyValues {Curve {keys [<at 0 1.5>,<at 0.4 1>,<at 1 0.968>]}}})
    Index 0 Param 2=1
    Index 0 Param 3=1
    Index 0 Param 4=1
    Module.Param0WriteEnabled=true
    Module.Param1WriteEnabled=true
    Module.Param2WriteEnabled=true
    Module.Param3WriteEnabled=true
   )
   &ScaleColor(
    Scale Alpha=&FloatFromCurve(FloatCurve=DataInterface() {propertyValues {Curve {keys [<at 0 0.5 RCIM_Cubic>,<at 1 0>]}}})
    Scale RGB=&MakeVectorFromLinearColorRGB(Color=&ColorFromCurve(ColorCurve=DataInterface() {propertyValues {
        AlphaCurve {keys [<at 0 1 RCIM_Cubic>]}
        BlueCurve {keys [<at 0 1 RCIM_Cubic>,<at 1 1 RCIM_Cubic>]}
        GreenCurve {keys [<at 0 1 RCIM_Cubic>,<at 1 0.5246 RCIM_Cubic>]}
        RedCurve {keys [<at 0 1 RCIM_Cubic>,<at 1 0 RCIM_Cubic>]}
       }}))
    ScaleA=true
    ScaleRGB=true
   )
   &ScaleMeshSize(Scale Factor=&VectorFromFloat(Value=&FloatFromCurve(
      FloatCurve=DataInterface() {propertyValues {Curve {keys [{interpMode=RCIM_Cubic, leaveTangent=1.6767, tangentMode=RCTM_Break, time=0, value=0}, <at 1 1.7>]}}}
      Scale Curve=&Multiply_Float(A=1.2, B=1)
     )))
   &CameraOffset(Camera Offset Amount=50)
  ]
  rendererClasses [&NiagaraMeshRendererProperties]
  renderers [{rendererClass=&NiagaraMeshRendererProperties, rendererIndex=0}]
 }
 Init_Glow001 {
  bEnabled=false
  emitterUpdateScript [
   &EmitterState(Life Cycle Mode=::NewEnumerator1, Loop Behavior=::NewEnumerator1, Loop Duration=1, UseLoopDelay=false)
   &SpawnBurst_Instantaneous(
    Spawn Count=&MakeCustomIntFromBool(
     Bool=$User.bIsFrontEndPreview
     True INT=&MakeCustomIntFromBool(Bool=$User.Return, True INT=0, False INT=1)
     False INT=1
    )
    Use Spawn Probability=false
   )
  ]
  particleSpawnScript [
   &InitializeParticle(
    Color=<rgba 0.5 0 0.3995 2>
    Lifetime=0.3
    Mesh Scale Mode=::NewEnumerator3
    Mesh Uniform Scale=2
    UsePositionOffset=false
    Write Material Random=true
   ) {module=InitializeParticle001}
   &InitialMeshOrientation(
    Rotation=&RandomRangeVector(Minimum=<0 0 0>, Maximum=<1 1 1>)
    Use Orientation Vector=true
    Use Rotation Vector=true
   ) {module=InitialMeshOrientation001}
  ]
  particleUpdateScript [
   &ParticleState()
   &DynamicMaterialParameters(
    Index 0 Param 1=&FloatFromCurve(FloatCurve=DataInterface() {propertyValues {Curve {keys [<at 0 1.5>,<at 0.4 1>,<at 1 0.968>]}}})
    Index 0 Param 2=1
    Index 0 Param 3=1
    Index 0 Param 4=1
    Module.Param0WriteEnabled=true
    Module.Param1WriteEnabled=true
    Module.Param2WriteEnabled=true
    Module.Param3WriteEnabled=true
   )
   &ScaleColor(
    Scale Alpha=&FloatFromCurve(FloatCurve=DataInterface() {propertyValues {Curve {keys [<at 0 0.5 RCIM_Cubic>,<at 1 0>]}}})
    Scale RGB=&MakeVectorFromLinearColorRGB(Color=&ColorFromCurve(ColorCurve=DataInterface() {propertyValues {
        AlphaCurve {keys [<at 0 1 RCIM_Cubic>]}
        BlueCurve {keys [<at 0 1 RCIM_Cubic>,<at 1 1 RCIM_Cubic>]}
        GreenCurve {keys [<at 0 1 RCIM_Cubic>,<at 1 0 RCIM_Cubic>]}
        RedCurve {keys [<at 0 1 RCIM_Cubic>,<at 1 0.2703 RCIM_Cubic>]}
       }}))
    ScaleA=true
    ScaleRGB=true
   )
   &ScaleMeshSize(Scale Factor=&VectorFromFloat(Value=&FloatFromCurve(
      FloatCurve=DataInterface() {propertyValues {Curve {keys [{interpMode=RCIM_Cubic, leaveTangent=1.6767, tangentMode=RCTM_Break, time=0, value=0}, <at 0.4 1.5>, <at 1 2>]}}}
      Scale Curve=&Multiply_Float(A=0.7, B=$System.Size Scale)
     )))
   &CameraOffset(Camera Offset Amount=50)
  ]
  rendererClasses [&NiagaraMeshRendererProperties]
  renderers [{rendererClass=&NiagaraMeshRendererProperties, rendererIndex=0}]
 }
 Light001 {
  emitterUpdateScript [
   &EmitterState(Life Cycle Mode=::NewEnumerator1, Loop Behavior=::NewEnumerator1, Loop Duration=0, UseLoopDelay=true)
   &SpawnBurst_Instantaneous(Spawn Count=1, Use Spawn Probability=false)
  ]
  particleSpawnScript [
   &InitializeParticle(
    Color=<rgba 0 2.9464 5 1>
    Lifetime=1.5
    Sprite Size Mode=::NewEnumerator1
    Uniform Sprite Size Max=5000
    Uniform Sprite Size Min=5000
    UsePositionOffset=false
    Write Material Random=true
   )
   &SystemLocation(MaskSpawnGroup=false)
   /SetVariables(Particles.LightPos:Vector3f=<0 0 500>) {enabled=false, isSetParameters=true}
  ]
  particleUpdateScript [
   &ParticleState()
   &CameraOffset(Camera Offset Amount=100)
   &ScaleColor(
    Scale Mode=::NewEnumerator1
    Scale RGBA=&MakeVector4_FromLinearColor(LinearColor=&LerpLinearColors(StartColor=<rgba 0.4891 0.3203 1 1>, EndColor=<rgba 0.2583 0.4538 2 1>, LerpFactor=0))
    ScaleRGBA=true
   )
   &FN_Light_Attributes(
    Color=<rgba 3.198 0 5 1>
    Create Color Attribute=true
    Create Exponent Attribute=true
    Create Radius Attribute=true
    Exponent=2
    Radius=&Multiply_Float(
     A=&FloatFromCurve(FloatCurve=DataInterface() {propertyValues {Curve {keys [<at 0 1>,<at 1 0>]}}})
     B=26000
    )
    Scale Color=50
   )
   &ScaleSpriteSize(
    Uniform Curve Scale=&Multiply_Float(A=1, B=$System.Size Scale)
    Uniform Curve Sprite Scale=DataInterface() {propertyValues {Curve {keys [{interpMode=RCIM_Cubic, tangentMode=RCTM_User, time=0, value=1}, {interpMode=RCIM_Cubic, tangentMode=RCTM_User, time=1, value=0}]}}}
   )
   &DynamicMaterialParameters(
    Index 0 Param 1=0.5
    Index 0 Param 2=1
    Index 1 Param 1=500
    Module.Param0WriteEnabled=true
    Module.Param0WriteEnabled1=true
    Module.Param1WriteEnabled=true
    Module.Param1WriteEnabled1=true
    Module.Param2WriteEnabled=true
    Module.Param2WriteEnabled1=true
    Module.Param3WriteEnabled=true
    Module.Param3WriteEnabled1=true
    Write Parameter Index 1=true
   )
  ]
  rendererClasses [&NiagaraSpriteRendererProperties, &NiagaraLightRendererProperties]
  renderers [
   {rendererClass=&NiagaraSpriteRendererProperties}
   {rendererClass=&NiagaraLightRendererProperties, rendererIndex=1}
  ]
 }
 Radial_Plasma_Energy_02 {
  emitterUpdateScript [
   &EmitterState()
   &SpawnBurst_Instantaneous(Spawn Count=1, Use Spawn Probability=false)
  ]
  particleSpawnScript [&InitializeParticle(
    Color Mode=::NewEnumerator0
    Lifetime=0.8
    Sprite Size Mode=::NewEnumerator3
    Uniform Sprite Size=22
    UsePositionOffset=false
    Write Material Random=true
   )]
  particleUpdateScript [
   &ParticleState()
   &ScaleSpriteSize() {module=ScaleSpriteSize004}
   &FN_Radial_Material(Base Sprite Tile V=2, Depth Fade=250, Radial Mask Power=10, Radial Mask Tile U=0, Sprite Extrusion Speed=0.6, Sprite Spin Speed=0.1)
   &Color(
    Color=&ColorFromCurve(ColorCurve=DataInterface() {propertyValues {
       AlphaCurve {keys [<at 0 1 RCIM_Cubic>,<at 1 0 RCIM_Cubic>]}
       BlueCurve {keys [<at 0 28.3799 RCIM_Cubic>]}
       GreenCurve {keys [<at 0 0 RCIM_Cubic>]}
       RedCurve {keys [<at 0 30 RCIM_Cubic>]}
      }})
    Scale Alpha=0.7
    Scale Color=&VectorFromFloat(Value=&MakeCustomFloatFromBool(Bool=$User.Hitting, True Float=2, False Float=1))
   )
   &ScaleSpriteSize(Scale Factor=&Vector2DFromFloat(Value=$System.GrowingWOffset)) {enabled=false, module=ScaleSpriteSize001}
   &ScaleSpriteSize(Scale Factor=&Vector2DFromFloat(Value=$System.Shrinking)) {enabled=false, module=ScaleSpriteSize002}
   &ScaleSpriteSize(Scale Factor=&Vector2DFromFloat(Value=&Lerp_Float(
      A=1
      B=&Lerp_Float(
       A=0.8
       B=1
       Alpha=&Waveform("Waveform [1]"=::NewEnumerator0, "[1] Frequency"=7, "[1] Amplitude Min/Max"=<0 1>, "[2] Phase"=$Emitter.Age, "[3] Phase"=$Emitter.Age)
      )
      Alpha=$System.FN_Timer.Value
     ))) {enabled=false}
   &ScaleSpriteSize(
    Scale Sprite Size Mode=::NewEnumerator0
    Uniform Scale Factor=&Multiply_Float(
     A=&Multiply_Float(A=$User.Scale, B=3000)
     B=&Lerp_Float(
      A=1
      B=&Lerp_Float(
       A=0.8
       B=1
       Alpha=&Waveform("Waveform [1]"=::NewEnumerator0, "[1] Frequency"=7, "[1] Amplitude Min/Max"=<0 1>, "[2] Phase"=$Emitter.Age, "[3] Phase"=$Emitter.Age)
      )
      Alpha=$System.FN_Timer.Value
     )
    )
   ) {module=ScaleSpriteSize003}
   &CameraOffset(Camera Offset Amount=30)
   &DynamicMaterialParameters(
    Index 0 Param 4=10
    Module.Index0Float0Write=true
    Module.Index0Float1Write=false
    Module.Index0Float2Write=false
    Module.Index0Float3Write=true
   )
  ]
  rendererClasses [&NiagaraSpriteRendererProperties]
  renderers [{rendererClass=&NiagaraSpriteRendererProperties}]
 }
 Sphere_ShockWave {
  emitterUpdateScript [
   &EmitterState(Life Cycle Mode=::NewEnumerator1, Loop Behavior=::NewEnumerator1, Loop Duration=1, UseLoopDelay=false)
   &SpawnBurst_Instantaneous(
    Spawn Count=&MakeCustomIntFromBool(
     Bool=$User.bIsFrontEndPreview
     True INT=&MakeCustomIntFromBool(Bool=$User.Return, True INT=0, False INT=1)
     False INT=1
    )
    Use Spawn Probability=false
    UseLoopCountLimit=false
   )
  ]
  particleSpawnScript [&InitializeParticle(
    Color=<rgba 2 2 2 2>
    Mesh Scale=&Multiply_VectorByFloat(Vector=<1 1 1>, Float=14)
    Mesh Scale Mode=::NewEnumerator4
    UsePositionOffset=false
    Write Material Random=true
   )]
  particleUpdateScript [
   &ParticleState()
   &ScaleMeshSize(Scale Factor=&Multiply_VectorByFloat(
     Vector=<1 1 1>
     Float=&FloatFromCurve(
      FloatCurve=DataInterface() {propertyValues {Curve {keys [
          {interpMode=RCIM_Cubic, tangentMode=RCTM_User, time=0, value=0}
          <at 0.4 0.7>
          {interpMode=RCIM_Cubic, tangentMode=RCTM_User, time=1, value=1}
         ]}}}
      Scale Curve=&Multiply_Float(A=1.4, B=$System.Size Scale)
     )
    ))
   &ScaleColor(
    Scale Alpha=&FloatFromCurve(
     FloatCurve=DataInterface() {propertyValues {Curve {keys [{interpMode=RCIM_Cubic, tangentMode=RCTM_User, time=0, value=1}, <at 0.6 1>, {interpMode=RCIM_Cubic, tangentMode=RCTM_User, time=1, value=0}]}}}
     Scale Curve=&Multiply_Float(A=1, B=1)
    )
    ScaleA=true
    ScaleRGB=true
   )
   &DynamicMaterialParameters(
    Index 0 Param 1=&FloatFromCurve(FloatCurve=DataInterface() {propertyValues {Curve {keys [{interpMode=RCIM_Cubic, tangentMode=RCTM_User, time=0, value=1}, {interpMode=RCIM_Cubic, tangentMode=RCTM_User, time=1, value=2}]}}})
    Index 0 Param 2=&FloatFromCurve(FloatCurve=DataInterface() {propertyValues {Curve {keys [<at 0 1 RCIM_Cubic>, {arriveTangent=-1, interpMode=RCIM_Cubic, leaveTangent=-1, time=0.5704, value=0.55}, <at 1 0>]}}})
    Module.Param0WriteEnabled=true
    Module.Param1WriteEnabled=true
    Module.Param2WriteEnabled=true
    Module.Param3WriteEnabled=true
   )
  ]
  rendererClasses [&NiagaraMeshRendererProperties]
  renderers [{rendererClass=&NiagaraMeshRendererProperties, rendererIndex=0}]
 }
 Wave02 {
  emitterUpdateScript [
   &EmitterState(Life Cycle Mode=::NewEnumerator1, Loop Behavior=::NewEnumerator1, Loop Duration=1, UseLoopDelay=true)
   &SpawnBurst_Instantaneous(Spawn Count=1, Use Spawn Probability=false)
  ]
  particleSpawnScript [
   &InitializeParticle(
    Color=<rgba 5 0 2.5246 1>
    Lifetime=0.8
    Mesh Scale Mode=::NewEnumerator3
    Mesh Uniform Scale=9
    UsePositionOffset=false
    Write Material Random=true
   )
   &InitialMeshOrientation(Orientation Axis=<0 0 1>, Orientation Vector=<0 0 1>, Rotation=<0 0.25 0>, Use Orientation Vector=true, Use Rotation Vector=true)
   &InitialMeshOrientation(
    Orientation Axis=<0 0 0>
    Orientation Vector=<0 0 0>
    Rotation=&RandomRangeVector(Minimum=<0 0 0>, Maximum=<1 0 0>)
    Use Orientation Vector=false
    Use Rotation Vector=true
   ) {module=InitialMeshOrientation001}
   &CameraOffset(Camera Offset Amount=100)
  ]
  particleUpdateScript [
   &ParticleState()
   &ScaleColor(
    Linear Color Curve=DataInterface() {propertyValues {
      AlphaCurve {keys [<at 0 1>,<at 0.3576 1>,<at 1 0>]}
      BlueCurve {keys [<at 0 0.0414>,<at 0.1684 0.45>,<at 0.3654 1>,<at 1 1>]}
      GreenCurve {keys [<at 0 1>,<at 0.1684 1>,<at 0.3654 0.701>,<at 1 0.0574>]}
      RedCurve {keys [<at 0 0.2012>,<at 0.1684 0>,<at 0.3654 0.1029>,<at 1 0.0414>]}
     }}
    Scale Mode=::NewEnumerator2
   ) {enabled=false}
   &DynamicMaterialParameters(
    Index 0 Param 1=&FloatFromCurve(FloatCurve=DataInterface() {propertyValues {Curve {keys [{arriveTangent=-2.1015, interpMode=RCIM_Cubic, leaveTangent=-2.1015, tangentMode=RCTM_User, time=0, value=0.5}, <at 1 -0.15>]}}})
    Index 0 Param 2=&FloatFromCurve(FloatCurve=DataInterface() {propertyValues {Curve {keys [{arriveTangent=-0.3464, interpMode=RCIM_Cubic, leaveTangent=-0.3464, tangentMode=RCTM_User, time=0, value=1}, <at 1 0.9>]}}})
    Index 0 Param 3=&FloatFromCurve(FloatCurve=DataInterface() {propertyValues {Curve {keys [<at 0 0>,<at 0.15 1>,<at 1 1.2>]}}})
    Module.Param0WriteEnabled=true
    Module.Param1WriteEnabled=true
    Module.Param2WriteEnabled=true
    Module.Param3WriteEnabled=false
   )
   &ScaleColor(
    Scale Alpha=&FloatFromCurve(FloatCurve=DataInterface() {propertyValues {Curve {keys [<at 0 1>,<at 1 0>]}}})
    Scale RGB=&VectorFromFloat(Value=&FloatFromCurve(FloatCurve=DataInterface() {propertyValues {Curve {keys [<at 0 10>,<at 0.15 3>,<at 1 1>]}}}))
    ScaleA=true
    ScaleRGB=true
   ) {module=ScaleColor001}
   &ScaleMeshSize(Scale Factor=&VectorFromFloat(Value=&FloatFromCurve(
      FloatCurve=DataInterface() {propertyValues {Curve {keys [<at 0 0>,<at 0.1 1.4>,<at 1 2>]}}}
      Scale Curve=&Multiply_Float(A=1, B=$System.Size Scale)
     )))
  ]
  rendererClasses [&NiagaraMeshRendererProperties]
  renderers [{rendererClass=&NiagaraMeshRendererProperties, rendererIndex=0}]
 }
}
summary {
 systemName=NS_TentacleEnergyBallDissipate
 emitters [
  {emitterName=Sphere_ShockWave, rendererClasses [&NiagaraMeshRendererProperties]}
  {emitterName=Explode_SphereDark, rendererClasses [&NiagaraMeshRendererProperties]}
  {bEnabled=false, emitterName=Init_Glow001, rendererClasses [&NiagaraMeshRendererProperties]}
  {emitterName=Light001, rendererClasses [&NiagaraSpriteRendererProperties, &NiagaraLightRendererProperties]}
  {emitterName=ExplodeFresnelSphere, rendererClasses [&NiagaraMeshRendererProperties]}
  {emitterName=Init_Glow, rendererClasses [&NiagaraMeshRendererProperties]}
  {emitterName=Wave02, rendererClasses [&NiagaraMeshRendererProperties]}
  {emitterName=Explode_Sparkles, rendererClasses [&NiagaraSpriteRendererProperties]}
  {emitterName=Glow, rendererClasses [&NiagaraSpriteRendererProperties]}
  {emitterName=BIg_Lightning, rendererClasses [&NiagaraMeshRendererProperties]}
  {emitterName=Radial_Plasma_Energy_02, rendererClasses [&NiagaraSpriteRendererProperties]}
 ]
 userVariables [
  User.bIsFrontEndPreview {type=NiagaraBool}
  User.ChargeUpLength {type=NiagaraFloat, defaultValue=0, description=""}
  User.Hitting {type=NiagaraBool}
  User.Return {type=NiagaraBool}
  User.Scale {type=NiagaraFloat, defaultValue=1, description=""}
 ]
}
